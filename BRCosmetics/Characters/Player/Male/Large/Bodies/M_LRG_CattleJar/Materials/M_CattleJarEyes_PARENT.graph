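# Parent eye material for the CattleJar character, built on shared character functions: PreFX, SkinFX, PostFX and cinematic PostFX.
# Vertex color red, amplified into a near-binary mask, blends two material-attribute branches.
# Zero-red vertices take SkinFX output, with world position offset collapsing them to the object pivot when UseTearsGeometry is zero.
# Red-marked vertices pass SkinFX through MF_CattleJar_Banana, overridden to a dark, glossy surface.
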Material /BRCosmetics/Characters/Player/Male/Large/Bodies/M_LRG_CattleJar/Materials/M_CattleJarEyes_PARENT
graph {
 properties {
  bAutomaticallySetUsageInEditor=false
  bUseMaterialAttributes=true
  bUsedWithMeshDeformer=true
  bUsedWithMorphTargets=true
  bUsedWithSkeletalMesh=true
  pixelDepthOffsetMode=PDOM_Legacy
  refractionMethod=RM_IndexOfRefraction
  shadingModel=MSM_FromMaterialExpression
  subsurfaceProfile=&SS_HeroSkin_02
  usedShadingModels=Default Lit | Subsurface | Clear Coat | Subsurface Profile | Cloth | Eye
 }
 nodes [
  n1 = MaterialFunctionCall() {materialFunction=&BaseFN_CharacterFunction_PreFX}
  n2 = MaterialFunctionCall(
   PreskinnedPosition=@n1.Pre skinned Local Position
   Gmap=@n1.Pre skinned Local Position
   MaterialAttributes=@n1.Pre skinned Local Position
   AO=@n1.Pre skinned Local Position
  ) {materialFunction=&CharacterSkinFX}
  n3 = WorldPosition()
  n4 = ObjectPositionWS()
  n5 = LinearInterpolate(A=Multiply(A=Subtract(A=@n3.XYZ, B=@n4)) {constB=-1}, Alpha=Ceil(None=ScalarParameter() {parameterName=UseTearsGeometry})) {constA=1000000000, constB=0}
  n6 = MaterialFunctionCall!(Gmap, Material=@n2.Result) {materialFunction=&/MF_CattleJar_Banana}
  n7 = Constant3Vector() {constant=<rgba 0.0194 0.0203 0.0194 1>}
  n8 = Constant() {r=0.14}
  n9 = Constant() {r=0.54}
  n10 = Multiply(A=VertexColor().R) {constB=1000}
  n11 = MaterialFunctionCall(
   AO=@n1.AO
   Distance Blend=@n1.AO
   Fresnel=@n1.AO
   Gmap=@n1.AO
   Pre Skinned Local Position=@n1.AO
   Material=BlendMaterialAttributes(
    A=SetMaterialAttributes!(MaterialAttributes=@n2.Result, World Position Offset=@n5) {attributeSetTypes [#1]}
    B=SetMaterialAttributes!(MaterialAttributes=@n6.Result, Base Color=@n7, Roughness=@n8, Specular=@n9) {attributeSetTypes [#2, #3, #4]}
    Alpha=Saturate(None=@n10)
   )
  ) {materialFunction=&BaseFN_CharacterFunction_PostFX}
  MaterialFunctionCall(In=@n11.Result) {materialFunction=&CINE_MF_CharPostFX}
 ]
}
