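# Material function that overlays an animated flipbook mouth on the CattleJar character, using Banana Winter parameters and textures, behind the UseCattleJarMouth static switch.
# Chained If nodes compare facial pose parameters to a threshold to select the frame, with separate open-jaw frames.
# Scaled, offset UVs feed two SubUV samples: one supplies color and masked alpha, a bump-offset one shapes tinted emissive.
# It outputs modified material attributes and Mouth_Alpha.
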MaterialFunction /BRCosmetics/Characters/Player/Male/Large/Bodies/M_LRG_CattleJar/Materials/MF_CattleJarMouth
graph {nodes [
  n1 = TextureSample() {
   samplerType=SAMPLERTYPE_Normal
   texture=/Game/Characters/Player/Male/Medium/Bodies/M_MED_Banana/Skins/Winter/Textures/Banana_Winter_Mouth_normal
  }
  ChannelMaskParameter(None=@n1.RGBA) {group=Banana Winter Mouth, parameterName=Banana Winter Mouth Channel}
  n2 = FunctionInput() {inputType=FunctionInput_MaterialAttributes}
  n3 = GetMaterialAttributes(None=@n2) {attributeGetTypes [#1, #2, #3]}
  n4 = NamedRerouteUsage()
  n5 = VectorParameter() {group=Banana Winter Mouth, parameterName=Mouth Emissive Color, defaultValue=<rgba 0.8646 0.8646 0.8646 0>}
  n6 = LinearInterpolate(
   A=ScalarParameter() {defaultValue=8, group=Banana Winter Mouth, parameterName=SubImages}
   B=AppendVector(
    A=ScalarParameter() {defaultValue=8, group=Banana Winter Mouth, parameterName=SubImages X}
    B=ScalarParameter() {defaultValue=8, group=Banana Winter Mouth, parameterName=SubImages Y}
   )
   Alpha=ScalarParameter() {group=Banana Winter Mouth, parameterName=useNonSquareTexture}
  )
  n7 = ScalarParameter() {defaultValue=0.5, group=Banana Winter Mouth, parameterName=Flipbook scale}
  n8 = AppendVector(A=@n7, B=Multiply(A=@n7, B=ScalarParameter() {defaultValue=1, group=Banana Winter Mouth, parameterName=Flipbook scale V relative}))
  n9 = TextureCoordinate()
  n10 = ScalarParameter() {group=Banana Winter Mouth, parameterName=Flipbook_Rotation}
  n11 = Add(
   A=LinearInterpolate(
    A=Rotator(Coordinate=@n9, Time=@n10)
    B=TextureCoordinate() {coordinateIndex=1}
    Alpha=ScalarParameter() {group=Banana Winter Mouth, parameterName=use Second UV Channel}
   )
   B=AppendVector(
    A=ScalarParameter() {defaultValue=0.125, group=Banana Winter Mouth, parameterName=Flipbook X}
    B=ScalarParameter() {group=Banana Winter Mouth, parameterName=Flipbook Y}
   )
  )
  n12 = LinearInterpolate(
   A=@n11
   B=BumpOffset(Coordinate=@n11, Height=NamedRerouteUsage()) {referencePlane=1}
   Alpha=ScalarParameter() {defaultValue=1, group=Banana Winter Mouth, parameterName=BumpOffset Intensity}
  )
  n13 = ComponentMask(None=MaterialFunctionCall!(
    Texture=NamedRerouteUsage()
    UVs=MaterialFunctionCall(UVs=@n12, Texture Scale=@n8) {materialFunction=&ScaleUVsByCenter}.UVs
    SubImages=@n6
    Frame=NamedRerouteUsage()
   ) {materialFunction=/Game/Weapons/WeaponSkins/MaterialFunction/CustomNumbers_SubUV}.RGB) {r=true}
  n14 = ScalarParameter() {defaultValue=0.5, group=Banana Winter Mouth, parameterName=Flipbook scale}
  n15 = ScalarParameter() {defaultValue=1, group=Banana Winter Mouth, parameterName=Flipbook scale V relative}
  n16 = MaterialFunctionCall(UVs=@n11, Texture Scale=AppendVector(A=@n14, B=Multiply(A=@n14, B=@n15))) {materialFunction=&ScaleUVsByCenter}
  n17 = MaterialFunctionCall(Texture=NamedRerouteUsage(), UVs=@n16.UVs, SubImages=@n6, Frame=NamedRerouteUsage()) {materialFunction=&CustomNumbers_SubUV}
  n18 = Multiply(A=Multiply(A=@n17.Alpha, B=@n16.0-1 mask), B=Constant() {r=1})
  n19 = Add(A=NamedRerouteUsage(), B=Multiply(A=Multiply(A=@n5.RGB, B=@n13), B=@n18))
  FunctionOutput(None=StaticSwitchParameter(
    True=SetMaterialAttributes!(MaterialAttributes=@n3.MaterialAttributes, BaseColor=@n4, EmissiveColor=@n19, Normal=@n3.MaterialAttributes) {attributeSetTypes [#1, #2, #3]}
    False=@n2
   ) {group=Skin FX, parameterName=UseCattleJarMouth})
  FunctionOutput(None=@n18) {outputName=Mouth_Alpha}
  n20 = ScalarParameter() {defaultValue=0.1, group=Banana, parameterName=THreshold}
  n21 = ScalarParameter() {defaultValue=1, group=Banana, parameterName=Default Face}
  n22 = If(
   A=Max(A=ScalarParameter() {group=Banana, parameterName=R_lip_corner_narrow_pose}, B=ScalarParameter() {parameterName=PeelyBlenderFace})
   B=@n20
   "A > B"=ScalarParameter() {defaultValue=3, parameterName=Narrow Frame}
   "A == B"=@n21
   "A < B"=@n21
  )
  n23 = If(
   A=ScalarParameter() {group=Banana, parameterName=R_frown_pose}
   B=@n20
   "A > B"=ScalarParameter() {defaultValue=4, parameterName=Frown Frame}
   "A == B"=@n22
   "A < B"=@n22
  )
  n24 = If(
   A=ScalarParameter() {group=Banana, parameterName=R_smile_pose}
   B=@n20
   "A > B"=ScalarParameter() {defaultValue=8, parameterName=Smile Mouth animation Frame}
   "A == B"=@n23
   "A < B"=@n23
  )
  Max(A=Max(A=@n22, B=@n23), B=@n24)
  NamedRerouteDeclaration(None=LinearInterpolate(A=NamedRerouteUsage(), B=@n17.RGB, Alpha=@n18)) {name=BaseColor Out, nodeColor=<rgba 0 1 0.1328 1>}
  NamedRerouteDeclaration(None=@n3.BaseColor) {name=Base Color In, nodeColor=<rgba 0.2422 1 0 1>}
  NamedRerouteDeclaration(None=@n3.EmissiveColor) {name=Emissive In, nodeColor=<rgba 0.7813 0 1 1>}
  n25 = ScalarParameter() {group=Banana, parameterName=Default Open Mouth}
  n26 = If(
   A=ScalarParameter() {group=Banana, parameterName=R_Frown_pose}
   B=@n20
   "A > B"=ScalarParameter() {defaultValue=10, parameterName=Mouth Open Frown Frame}
   "A == B"=@n25
   "A < B"=@n25
  )
  n27 = If(
   A=ScalarParameter() {group=Banana, parameterName=R_smile_pose}
   B=@n20
   "A > B"=ScalarParameter() {defaultValue=6, parameterName=Mouth Open Smile Frame}
   "A == B"=@n25
   "A < B"=@n25
  )
  NamedRerouteDeclaration(None=If(A=ScalarParameter() {group=Banana, parameterName=Jaw_open_Pose}, B=@n20, "A > B"=Max(A=@n26, B=@n27), "A == B"=@n24, "A < B"=@n24)) {name=Frame, nodeColor=<rgba 1 0 0.0965 1>}
  n28 = TextureObjectParameter() {
   group=Banana Winter Mouth
   parameterName=Flipbook
   samplerType=SAMPLERTYPE_LinearColor
   texture=/Game/Characters/Player/Male/Medium/Bodies/M_MED_Banana/Skins/Winter/Textures/BananaWinterFaces
  }
  NamedRerouteDeclaration(None=@n28) {name=Texture Object, nodeColor=<rgba 0.2656 0 1 1>}
  VectorParameter() {group=Banana Winter Mouth, parameterName=Mouth Base Color, defaultValue=<rgba 0.1907 0.5191 0.75 0>}
 ]}
