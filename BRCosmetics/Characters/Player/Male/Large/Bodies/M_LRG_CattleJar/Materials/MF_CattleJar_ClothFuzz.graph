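# Material function for the CattleJar character body that adds cloth fuzz and cloth shading to incoming material attributes.
# Tiled fuzz texture times view-angle falloff drives a tinted BaseColor blend, optionally colored by RGB mask, whose blend alpha sums R twice and omits G.
# By default, a quality switch selects the Cloth shading model with Gmap-driven specular on Epic and a simple tint on lower tiers.
# Static switches can select the fuzz-only blend instead.
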MaterialFunction /BRCosmetics/Characters/Player/Male/Large/Bodies/M_LRG_CattleJar/Materials/MF_CattleJar_ClothFuzz
graph {nodes [
  n1 = GetMaterialAttributes(None=FunctionInput() {inputName=Material, inputType=FunctionInput_MaterialAttributes}) {attributeGetTypes [#1, #2, #3]}
  n2 = TextureSampleParameter2D() {
   group=Cloth
   parameterName=Cloth Fuzz Color Mask
   samplerSource=SSM_Wrap_WorldGroupSettings
   samplerType=SAMPLERTYPE_Masks
   texture=&T_BlackMask
  }
  n3 = LinearInterpolate(
   B=VectorParameter() {group=Cloth, parameterName=Fuzz Tint R Channel, defaultValue=<rgba 0.6823 0.6823 0.6823 1>}.RGB
   Alpha=ComponentMask(None=@n2.RGB) {r=true}
  )
  n4 = VectorParameter() {group=Cloth, parameterName=Fuzz Tint G Channel, defaultValue=<rgba 0.6823 0.6823 0.6823 1>}
  n5 = ComponentMask(None=@n2.RGB) {g=true}
  n6 = StaticBoolParameter() {group=Cloth, parameterName=Color Cloth Fuzz by Channel}
  n7 = StaticSwitch(
   True=LinearInterpolate(
    A=LinearInterpolate(A=@n3, B=@n4.RGB, Alpha=@n5)
    B=VectorParameter() {group=Cloth, parameterName=Fuzz Tint B Channel, defaultValue=<rgba 0.6823 0.6823 0.6823 1>}.RGB
    Alpha=ComponentMask(None=@n2.RGB) {b=true}
   )
   False=Multiply(A=@n1.BaseColor, B=VectorParameter() {group=Cloth, parameterName=Fuzz Tint, defaultValue=<rgba 0.6823 0.6823 0.6823 1>}.RGB)
   Value=@n6
  )
  n8 = DotProduct(A=PixelNormalWS(), B=CameraVectorWS())
  n9 = Clamp(None=Power(Base=OneMinus(None=@n8), Exp=ScalarParameter() {defaultValue=4, group=Cloth, parameterName=Fuzz Exponent}))
  n10 = TextureSampleParameter2D(UVs=Multiply(A=TextureCoordinate(), B=ScalarParameter() {defaultValue=4, group=Cloth, parameterName=Fuzz Tiling})) {
   group=Cloth
   parameterName=ClothFuzz Texture
   samplerSource=SSM_Wrap_WorldGroupSettings
   samplerType=SAMPLERTYPE_LinearColor
   texture=/Game/Global/Textures/Default/Masks_and_Gradients/T_Fuzz_MASK
  }
  n11 = Saturate(None=Multiply(A=Multiply(A=@n9, B=@n10.R), B=ScalarParameter() {group=Cloth, parameterName=Fuzz Fresnel Blend}))
  n12 = FunctionInput() {bUsePreviewValueAsDefault=true, inputName=Alpha, inputType=FunctionInput_Scalar}
  n13 = Add(A=Add(A=@n2.R, B=@n2.R), B=@n2.B)
  n14 = BlendMaterialAttributes(
   A=@n1.MaterialAttributes
   B=SetMaterialAttributes!(MaterialAttributes=@n1.MaterialAttributes, BaseColor=LinearInterpolate(A=@n1.BaseColor, B=@n7, Alpha=@n11)) {attributeSetTypes [#1]}
   Alpha=StaticSwitch(True=Saturate(None=@n13), False=@n12, Value=@n6)
  )
  n15 = VectorParameter() {group=Cloth, parameterName=Cloth Fuzz Tint, defaultValue=<rgba 1 1 1 1>}
  n16 = Multiply(A=@n1.BaseColor, B=@n15.RGB)
  n17 = BlendMaterialAttributes(
   A=@n1.MaterialAttributes
   B=SetMaterialAttributes(MaterialAttributes=@n1.MaterialAttributes, BaseColor=Saturate(None=@n16)) {attributeSetTypes [#1]}
   Alpha=@n12
  )
  n18 = Multiply(A=@n1.BaseColor, B=ScalarParameter() {defaultValue=1, group=Cloth, parameterName=Cloth Subsurface Intensity})
  n19 = ScalarParameter() {defaultValue=0.4, group=06_Cloth_SubstrateFix, parameterName=SubstrateClothFix_RoughnessBias, sliderMax=1, sortPriority=0}
  n20 = Saturate(None=Multiply(A=@n1.Roughness, B=ScalarParameter() {defaultValue=1, group=Cloth, parameterName=Cloth Roughness}))
  n21 = Multiply(A=@n1.BaseColor, B=ScalarParameter() {defaultValue=0.9, group=Cloth, parameterName=Cloth Base Color Intensity})
  n22 = Multiply(A=@n15.RGB, B=Add(A=Multiply(A=@n11) {constB=10}))
  n23 = LinearInterpolate(
   A=ScalarParameter() {defaultValue=0.75, group=Cloth, parameterName=Cloth Scattering Min}
   B=ScalarParameter() {defaultValue=1, group=Cloth, parameterName=Cloth Scattering Max}
   Alpha=FunctionInput() {inputName=Gmap, inputType=FunctionInput_Scalar}
  )
  n24 = ScalarParameter() {defaultValue=5, group=06_Cloth_SubstrateFix, parameterName=SubstrateClothFix_SpecularFresnelExponent, sortPriority=0}
  n25 = ScalarParameter() {defaultValue=0.3, group=06_Cloth_SubstrateFix, parameterName=SubstrateClothFix_SpecularFresnelBase, sortPriority=0}
  n26 = SetMaterialAttributes!(
   MaterialAttributes=@n1.MaterialAttributes
   ShadingModel=ShadingModel() {shadingModel=MSM_Cloth}
   SubsurfaceColor=Saturate(None=Multiply(A=@n15.RGB, B=@n18))
   Roughness=LinearInterpolate(A=@n19, Alpha=Saturate(None=@n20))
   CustomData0=ScalarParameter() {defaultValue=1, group=Cloth, parameterName=Cloth Shader Intensity}
   Opacity=Constant() {r=1}
   BaseColor=Saturate(None=Multiply(A=@n21, B=@n22))
   Specular=Multiply(A=Saturate(None=@n23), B=Fresnel(ExponentIn=@n24, BaseReflectFractionIn=@n25) {baseReflectFraction=0})
   Normal=@n1.MaterialAttributes
  ) {attributeSetTypes [
    #4
    #5
    #3
    #6
    #7
    #1
    #8
    #2
   ]}
  n27 = BlendMaterialAttributes(A=@n1.MaterialAttributes, B=@n26, Alpha=@n12)
  n28 = MaterialFunctionCall!(
   Low _ PC Low and Med=@n17
   "High _ XB1, PS4"=@n17
   "Epic _ PC High, XSX, PS5"=@n27
   --------
   "Mobile Low _ PC Perf, Android"=@n17
   Mobile Med _ Switch=@n17
  ) {materialFunction=/Game/Global/MaterialFunctions/TechArt/MF_Fortnite_QualitySwitch_MaterialAttributes}
  n29 = ShadingPathSwitch(Default=@n14, Mobile=@n1.MaterialAttributes)
  FunctionOutput(None=StaticSwitchParameter(True=@n14, False=StaticSwitchParameter(True=@n28.Result, False=@n29) {defaultValue=true, group=Cloth, parameterName=useNewClothShading}) {group=Cloth, parameterName=useFullClothShadingOnMobile})
  ShadingPathSwitch(Default=@n27, Mobile=@n17)
 ]}
