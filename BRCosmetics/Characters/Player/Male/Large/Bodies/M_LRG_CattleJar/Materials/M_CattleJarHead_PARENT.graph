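# Parent skeletal-mesh material for the CattleJar male large character's head, using material attributes and a hero skin subsurface profile.
# Shared character PreFX feeds skin FX, then a custom mouth function.
# A static switch optionally adds custom cloth fuzz, masked by a selectable skin FX mask channel and excluded from the mouth via its inverted alpha.
# The result passes through shared character PostFX and a cinematic PostFX function.
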Material /BRCosmetics/Characters/Player/Male/Large/Bodies/M_LRG_CattleJar/Materials/M_CattleJarHead_PARENT
graph {
 properties {
  bAutomaticallySetUsageInEditor=false
  bUseMaterialAttributes=true
  bUsedWithMeshDeformer=true
  bUsedWithMorphTargets=true
  bUsedWithSkeletalMesh=true
  pixelDepthOffsetMode=PDOM_Legacy
  refractionMethod=RM_IndexOfRefraction
  shadingModel=MSM_FromMaterialExpression
  subsurfaceProfile=&SS_HeroSkin_02
  usedShadingModels=Default Lit | Subsurface | Clear Coat | Subsurface Profile | Cloth | Eye
 }
 nodes [
  n1 = MaterialFunctionCall() {materialFunction=&BaseFN_CharacterFunction_PreFX}
  n2 = MaterialFunctionCall(
   PreskinnedPosition=@n1.Pre skinned Local Position
   Gmap=@n1.Pre skinned Local Position
   MaterialAttributes=@n1.Pre skinned Local Position
   AO=@n1.Pre skinned Local Position
  ) {materialFunction=&CharacterSkinFX}
  n3 = MaterialFunctionCall!(In=@n2.Result) {materialFunction=&/MF_CattleJarMouth}
  n4 = Multiply(
   A=OneMinus(None=@n3.Mouth_Alpha)
   B=StaticComponentMaskParameter(None=@n2.SkinFX_Mask) {defaultR=true, group=Skin FX, parameterName=Cloth Channel}
  )
  n5 = MaterialFunctionCall(
   AO=@n1.AO
   Distance Blend=@n1.AO
   Fresnel=@n1.AO
   Gmap=@n1.AO
   Pre Skinned Local Position=@n1.AO
   Material=StaticSwitchParameter(
    True=MaterialFunctionCall!(Alpha=@n4, Gmap=@n1.Gmap, Material=@n3.Result) {materialFunction=&/MF_CattleJar_ClothFuzz}.Result
    False=@n3.Result
   ) {parameterName=UseCustomClothFuzz}
  ) {materialFunction=&BaseFN_CharacterFunction_PostFX}
  MaterialFunctionCall(In=@n5.Result) {materialFunction=&CINE_MF_CharPostFX}
  MaterialFunctionCall!(In=TextureObjectParameter() {group=Skin FX, parameterName=SkinFX_Mask}) {materialFunction=/Game/Weapons/WeaponSkins/MaterialFunction/MF_GlitchSubfunction}
 ]
}
