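# Control Rig for the CattleJar large male player body that drives corrective deform bones as the skeleton poses.
# A shared DeformRig function, configured with per-region toggles and offsets, runs first, then a custom knee pass follows.
# Knee base bones are parent-constrained between thigh and calf, while calf yaw from reference pose, remapped and clamped, weights the Knee_in/Knee_out offsets, mirrored for the right leg.
# Branches writing knee-alpha anim attributes are left unconnected.
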ControlRigBlueprint /BRCosmetics/Characters/Player/Male/Large/Bodies/M_LRG_CattleJar/Meshes/M_LRG_CattleJar_CtrlRig
graphs [
 ContainedGraph {nodes [
   n1 = RigVMFunctionEntryNode()
   SequenceExecution = &RigVMFunction_Sequence(ExecuteContext=@n1.ExecuteContext)
   SequenceExecution_1 = &RigVMFunction_Sequence(ExecuteContext=@SequenceExecution.B)
   SequenceExecution_2 = &RigVMFunction_Sequence(ExecuteContext=@SequenceExecution_1.B)
   RigVMFunctionReturnNode(A=@SequenceExecution.A, B=@SequenceExecution_1.A, C=@SequenceExecution_2.A, D=@SequenceExecution_2.B)
  ]}
 ContainedGraph {nodes [
   n1 = RigVMFunctionEntryNode()
   SequenceExecution_2 = &RigVMFunction_Sequence(ExecuteContext=@n1.ExecuteContext)
   SequenceExecution_2_1 = &RigVMFunction_Sequence(ExecuteContext=@SequenceExecution_2.B)
   SequenceExecution_2_2 = &RigVMFunction_Sequence(ExecuteContext=@SequenceExecution_2_1.B)
   RigVMFunctionReturnNode(A=@SequenceExecution_2.A, B=@SequenceExecution_2_1.A, C=@SequenceExecution_2_2.A, D=@SequenceExecution_2_2.B)
  ]}
 ContainedGraph {nodes [
   n1 = RigVMFunctionEntryNode()
   SequenceExecution_1_1 = &RigVMFunction_Sequence(ExecuteContext=@n1.ExecuteContext)
   SequenceExecution_1_1_1 = &RigVMFunction_Sequence(ExecuteContext=@SequenceExecution_1_1.B)
   SequenceExecution_1_1_2 = &RigVMFunction_Sequence(ExecuteContext=@SequenceExecution_1_1_1.B)
   RigVMFunctionReturnNode(A=@SequenceExecution_1_1.A, B=@SequenceExecution_1_1_1.A, C=@SequenceExecution_1_1_2.A, D=@SequenceExecution_1_1_2.B)
  ]}
 ContainedGraph {nodes [
   n1 = RigVMFunctionEntryNode()
   SequenceExecution_1 = &RigVMFunction_Sequence(ExecuteContext=@n1.ExecuteContext)
   SequenceExecution_1_1 = &RigVMFunction_Sequence(ExecuteContext=@SequenceExecution_1.B)
   SequenceExecution_1_2 = &RigVMFunction_Sequence(ExecuteContext=@SequenceExecution_1_1.B)
   RigVMFunctionReturnNode(A=@SequenceExecution_1.A, B=@SequenceExecution_1_1.A, C=@SequenceExecution_1_2.A, D=@SequenceExecution_1_2.B)
  ]}
 ContainedGraph {nodes [
   n1 = RigVMFunctionEntryNode()
   n2 = /Script/RigVM.RigVMFunction_MathTransformToEulerTransform(Value=@n1.transform) {is=struct RigVMUnitNode, template="Cast::Execute(in Value,out Result)"}
   > RigVMFunctionReturnNode(
    mirrorTransform=(Rotation=(X=0,Y=0,Z=0,W=1),Translation=(X=0,Y=0,Z=0),Scale3D=(X=0,Y=0,Z=0))
    mirrorTransform.Rotation=/Script/RigVM.RigVMFunction_MathQuaternionFromRotatorV2(Value=@n2.Result.Rotation) {is=struct RigVMUnitNode, template="Cast::Execute(in Value,out Result)"}.Result
    mirrorTransform.Translation=/Script/RigVM.RigVMFunction_MathVectorScale(Factor=-1, Value=@n2.Result.Location) {is=struct RigVMUnitNode, template="Scale::Execute(in Value,in Factor,out Result)"}.Result
    mirrorTransform.Scale3D=@n2.Result.Scale
   )
  ]}
 ContainedGraph {nodes [
   n1 = RigVMFunctionEntryNode()
   n2 = &RigUnit_GetTransform(Space=LocalSpace, Item=@n1.Item)
   n3 = &RigUnit_GetTransform(Space=LocalSpace, bInitial=true, Item=@n1.Item)
   > RigVMFunctionReturnNode(OutTransform=&RigVMFunction_MathTransformToEulerTransform(Value=&RigVMFunction_MathTransformMakeRelative(Global=@n2.Transform, Parent=@n3.Transform).Local) {template="Cast::Execute(in Value,out Result)"}.Result)
  ]}
 RigVMFunctionLibrary {nodes [
   DeformRig_GetDeltaRefTransform_FNC = RigVMCollapseNode(ExecuteContext=(), Item=(Type=Bone))
   DeformRig_GetMirrorTransform_FNC = RigVMCollapseNode(ExecuteContext=(), transform=(Rotation=(X=0,Y=0,Z=0,W=1),Translation=(X=0,Y=0,Z=0),Scale3D=(X=1,Y=1,Z=1)))
  ]}
 RigVMModel {nodes [
   RigUnit_BeginExecution = &RigUnit_BeginExecution()
   RigVMFunction_Sequence = &RigVMFunction_Sequence(ExecuteContext=@RigUnit_BeginExecution.ExecutePin)
   FN_DeformRig_FNC = RigVMFunctionReferenceNode(
    DeformRig_ Curve="(deformRig_l_upperarm_up_alpha_72_#1=False,deformRig_r_upperarm_up_alpha_73_#2=False,deformRig_l_upperarm_fr_alpha_74_#3=False,deformRig_r_upperarm_fr_alpha_75_CB51B8 ...(truncated)"
    DeformRig_Bool="(upperarm_1_#4=True,elbow_3_#5=True,wrist_7_#6=True,groin_8_#7=True,glute_17_#8=True,knee_9_40C408324CA261 ...(truncated)"
    DeformRig_Debug=(upperarm_debug_18_#4=False,wrist_debug_20_#6=False,lat_debug_26_#9=False,groinGlute_debug_27_#7=False)
    DeformRig_Value="(upperarm_up_18_#10=(Rotation=(X=0.0,Y=-0.0,Z=0.0,W=1.0),Translation=(X=0.0,Y=0.0,Z=8.0),Scale3D=(X=0.70,Y=1.0,Z=1.0)),upperarm_fwd_fr_19_#11=(Rotation=(X=0.0,Y=-0.0,Z=0.0,W=1.0),Translation=(X=0.0, ...(truncated)"
    ExecuteContext=@RigVMFunction_Sequence.A
   )
   SequenceExecution_2 = RigVMAggregateNode(ExecuteContext=@RigVMFunction_Sequence.B)
   DeformRig_GetDeltaRefTransform_FNC = RigVMFunctionReferenceNode(Item=(Type=Bone,Name="calf_l"), ExecuteContext=@SequenceExecution_2.A)
   DeformRig_GetDeltaRefTransform_FNC_1 = RigVMFunctionReferenceNode(Item=(Type=Bone,Name="calf_r"), ExecuteContext=@SequenceExecution_2.B)
   SequenceExecution = RigVMAggregateNode(ExecuteContext=@SequenceExecution_2.C)
   SequenceExecution_1 = RigVMAggregateNode(ExecuteContext=@SequenceExecution.A)
   &RigUnit_ParentConstraint(
    AdvancedSettings=(RotationOrderForFilter=XZY)
    Child=(Type=Bone,Name="deform_knee_in_base_l")
    Parents=((Item=(Type=Bone,Name="calf_l"),Weight=0.5),(Item=(Type=Bone,Name="thigh_l"),Weight=0.5))
    Weight=1
    ExecutePin=@SequenceExecution_1.A
   )
   n1 = RigVMVariableNode() {variable=Knee_in}
   n2 = &RigVMFunction_MathFloatRemap(SourceMaximum=100, SourceMinimum=10, bClamp=true, Value=@DeformRig_GetDeltaRefTransform_FNC.OutTransform.Rotation.Yaw)
   > &RigUnit_ModifyTransforms(
    ItemToModify=((Item=(Type=Bone,Name="deform_knee_in_l"),Transform=(Rotation=(X=0,Y=0.1045,Z=0,W=0.9945),Translation=(X=0,Y=0,Z=0),Scale3D=(X=0.5,Y=1,Z=1))))
    ItemToModify.0.Transform=@n1.Value
    Weight=@n2.Result
   )
   &RigUnit_ParentConstraint(
    AdvancedSettings=(RotationOrderForFilter=XZY)
    Child=(Type=Bone,Name="deform_knee_in_base_r")
    Parents=((Item=(Type=Bone,Name="calf_r"),Weight=0.5),(Item=(Type=Bone,Name="thigh_r"),Weight=0.5))
    Weight=1
    ExecutePin=@SequenceExecution_1.B
   )
   > DeformRig_GetMirrorTransform_FNC = RigVMFunctionReferenceNode(transform=@n1.Value)
   n3 = &RigVMFunction_MathFloatRemap(SourceMaximum=100, SourceMinimum=10, bClamp=true, Value=@DeformRig_GetDeltaRefTransform_FNC_1.OutTransform.Rotation.Yaw)
   > &RigUnit_ModifyTransforms(
    ItemToModify=((Item=(Type=Bone,Name="deform_knee_in_r"),Transform=(Rotation=(X=0,Y=0,Z=0,W=1),Translation=(X=0,Y=0,Z=0),Scale3D=(X=1,Y=1,Z=1))))
    ItemToModify.0.Transform=@DeformRig_GetMirrorTransform_FNC.mirrorTransform
    Weight=@n3.Result
   )
   SequenceExecution_1_1 = RigVMAggregateNode(ExecuteContext=@SequenceExecution.B)
   &RigUnit_ParentConstraint(
    AdvancedSettings=(RotationOrderForFilter=XZY)
    Child=(Type=Bone,Name="deform_knee_out_base_l")
    Parents=((Item=(Type=Bone,Name="calf_l"),Weight=0.1),(Item=(Type=Bone,Name="thigh_l"),Weight=0.9))
    Weight=1
    ExecutePin=@SequenceExecution_1_1.A
   )
   n4 = RigVMVariableNode() {variable=Knee_out}
   > &RigUnit_ModifyTransforms(
    ItemToModify=((Item=(Type=Bone,Name="deform_knee_out_l"),Transform=(Rotation=(X=0,Y=-0.0262,Z=0,W=0.9997),Translation=(X=0,Y=0,Z=0),Scale3D=(X=1.5,Y=1,Z=1.5))))
    ItemToModify.0.Transform=@n4.Value
    Weight=@n2.Result
   )
   &RigUnit_ParentConstraint(
    AdvancedSettings=(RotationOrderForFilter=XZY)
    Child=(Type=Bone,Name="deform_knee_out_base_r")
    Parents=((Item=(Type=Bone,Name="calf_r"),Weight=0.1),(Item=(Type=Bone,Name="thigh_r"),Weight=0.9))
    Weight=1
    ExecutePin=@SequenceExecution_1_1.B
   )
   > DeformRig_GetMirrorTransform_FNC_1 = RigVMFunctionReferenceNode(transform=@n4.Value)
   > &RigUnit_ModifyTransforms(
    ItemToModify=((Item=(Type=Bone,Name="deform_knee_out_r"),Transform=(Rotation=(X=0,Y=0,Z=0,W=1),Translation=(X=0,Y=0,Z=0),Scale3D=(X=1,Y=1,Z=1))))
    ItemToModify.0.Transform=@DeformRig_GetMirrorTransform_FNC_1.mirrorTransform
    Weight=@n3.Result
   )
   RigVMFunction_ControlFlowBranch = &RigVMFunction_ControlFlowBranch(ExecuteContext=())
   RigVMDispatchNode(BoneName=root, Name=deformRig_l_knee_alpha, ExecuteContext=@RigVMFunction_ControlFlowBranch.True, Value=@n2.Result) {template="DISPATCH_RigDispatch_SetAnimAttribute(in Name,in BoneName,in Value,out Success)"}
   n5 = &RigVMFunction_ControlFlowBranch(ExecuteContext=())
   RigVMDispatchNode(BoneName=root, Name=deformRig_r_knee_alpha, ExecuteContext=@n5.True, Value=@n3.Result) {template="DISPATCH_RigDispatch_SetAnimAttribute(in Name,in BoneName,in Value,out Success)"}
  ]}
]
hierarchy [
 root {loc=<0 0 0>, rot=<0 0 0 1>}
 pelvis {parent=root, loc=<0 -0.7321 92.4218>, rot=<0.0127 0.707 -0.0127 -0.707>}
 spine_01 {parent=pelvis, loc=<5.6002 -0.5128 0>, rot=<0 0 0.034 -0.9994>}
 spine_02 {parent=spine_01, loc=<8.0753 0.0139 0>, rot=<0 0 0.0238 -0.9997>}
 spine_03 {parent=spine_02, loc=<7.0393 0 0>, rot=<0 0 0.0149 -0.9999>}
 spine_04 {parent=spine_03, loc=<8.1024 0 0>, rot=<0 0 -0.0132 -0.9999>}
 spine_05 {parent=spine_04, loc=<10.0146 0 0>, rot=<0 0 -0.0015 -1>}
 neck_01 {parent=spine_05, loc=<10.815 1.9196 0>, rot=<0 0 -0.1864 -0.9825>}
 neck_02 {parent=neck_01, loc=<4.7347 0.3045 0>, rot=<0 0 0.0363 -0.9993>}
 head {parent=neck_02, loc=<5.7632 0.5552 0>, rot=<0 0 0.115 -0.9934>}
 attach_backpack {parent=spine_05, loc=<-0.123 -15.1423 0>, rot=<-0.0093 -0.707 0.0093 -0.707>}
 clavicle_l {parent=spine_05, loc=<2.5856 -0.2866 -4.3777>, rot=<0.0465 -0.7289 -0.009 -0.6829>}
 upperarm_l {parent=clavicle_l, loc=<18.2264 0.1342 -0.001>, rot=<0.0253 -0.2693 0.0013 -0.9627>}
 lowerarm_l {parent=upperarm_l, loc=<29.0071 0.1686 -1.239>, rot=<0.0391 -0.0099 -0.2453 -0.9686>}
 hand_l {parent=lowerarm_l, loc=<25.8607 0.222 1.0274>, rot=<-0.7193 -0.0988 0.1193 -0.6773>}
 thumb_01_l {parent=hand_l, loc=<1.3118 -0.3191 -2.459>, rot=<0.6546 -0.4042 0.0664 -0.6354>}
 thumb_02_l {parent=thumb_01_l, loc=<5.9336 -0.0448 0.0401>, rot=<-0.0037 0.0213 0.298 -0.9543>}
 thumb_03_l {parent=thumb_02_l, loc=<2.2223 0.3682 -0.0907>, rot=<0.0118 -0.013 0.0066 -0.9998>}
 index_metacarpal_l {parent=hand_l, loc=<1.6535 0.0183 -2.96>, rot=<0.0158 -0.0124 -0.0173 -0.9996>}
 index_01_l {parent=index_metacarpal_l, loc=<8.3008 0.0357 -0.0524>, rot=<-0.0044 0.0485 0.0924 -0.9945>}
 index_02_l {parent=index_01_l, loc=<3.3587 -0.3004 0.0452>, rot=<-0.002 0.0027 0.3225 -0.9466>}
 index_03_l {parent=index_02_l, loc=<2.5578 -0.123 -0.0252>, rot=<-0.002 0.0067 0.2558 -0.9667>}
 middle_metacarpal_l {parent=hand_l, loc=<2.1856 -0.1725 -0.5984>, rot=<0.006 0.0168 -0.0329 -0.9993>}
 middle_01_l {parent=middle_metacarpal_l, loc=<7.3275 -0.2447 0.0129>, rot=<-0.0208 0.0042 0.0858 -0.9961>}
 middle_02_l {parent=middle_01_l, loc=<4.0123 0.1401 0.1175>, rot=<-0.0218 0.0258 0.3512 -0.9357>}
 middle_03_l {parent=middle_02_l, loc=<2.6077 0.0534 0.1373>, rot=<0.0082 0.0188 0.1708 -0.9851>}
 ring_metacarpal_l {parent=hand_l, loc=<2.2074 -0.3625 1.3638>, rot=<-0.0284 0.0722 -0.0458 -0.9959>}
 ring_01_l {parent=ring_metacarpal_l, loc=<7.0295 -0.6197 -0.1446>, rot=<-0.0489 -0.0252 0.1892 -0.9804>}
 ring_02_l {parent=ring_01_l, loc=<3.6174 0.4647 -0.1027>, rot=<0 0 0.2806 -0.9598>}
 ring_03_l {parent=ring_02_l, loc=<2.4152 0.114 0.0199>, rot=<0 0 0.1189 -0.9929>}
 pinky_metacarpal_l {parent=hand_l, loc=<1.7616 -0.9753 3.3814>, rot=<-0.0397 0.0814 -0.069 -0.9935>}
 pinky_01_l {parent=pinky_metacarpal_l, loc=<7.1709 -0.4831 -0.1094>, rot=<-0.0654 -0.0035 0.2242 -0.9723>}
 pinky_02_l {parent=pinky_01_l, loc=<2.8274 0.0893 -0.1043>, rot=<0.0037 0.0011 0.2854 -0.9584>}
 pinky_03_l {parent=pinky_02_l, loc=<2.2046 0.0307 -0.0078>, rot=<-0.0002 0.0098 0.1445 -0.9895>}
 deform_wrist_dn_base_l {parent=hand_l, loc=<0 -3.4173 0>, rot=<0 0 0 -1>}
 deform_wrist_dn_l {parent=deform_wrist_dn_base_l, loc=<0 0 0>, rot=<0 0 0 1>}
 deform_wrist_up_base_l {parent=hand_l, loc=<0 2.7805 0>, rot=<0 0 0 -1>}
 deform_wrist_up_l {parent=deform_wrist_up_base_l, loc=<0 0 0>, rot=<0 0 0 1>}
 lowerarm_twist_02_l {parent=lowerarm_l, loc=<9.6199 0 0>, rot=<0 0 0 -1>}
 lowerarm_twist_01_l {parent=lowerarm_l, loc=<17.8549 0 0>, rot=<0 0 0 -1>}
 deform_elbow_in_base_l {parent=lowerarm_l, loc=<0.2112 5.645 0.8274>, rot=<0 0 0 -1>}
 deform_elbow_in_l {parent=deform_elbow_in_base_l, loc=<0 0 0>, rot=<0 0 0 1>}
 deform_elbow_out_base_l {parent=lowerarm_l, loc=<0 -7.5129 0>, rot=<0 0 0 -1>}
 deform_elbow_out_l {parent=deform_elbow_out_base_l, loc=<0 0 0>, rot=<0 0 0 1>}
 upperarm_twist_01_l {parent=upperarm_l, loc=<7.9796 0.0843 -0.6195>, rot=<0 0 0 -1>}
 deform_upperarm_bk_base_l {parent=upperarm_twist_01_l, loc=<-6.8983 -13.3826 -1.3258>, rot=<0 0 0 -1>}
 deform_upperarm_bk_l {parent=deform_upperarm_bk_base_l, loc=<0 0 0>, rot=<0 0 0 1>}
 deform_upperarm_fr_base_l {parent=upperarm_twist_01_l, loc=<-8.3487 10.7117 -2.1455>, rot=<0 0 0 -1>}
 deform_upperarm_fr_l {parent=deform_upperarm_fr_base_l, loc=<0 0 0>, rot=<0 0 0 1>}
 deform_upperarm_up_base_l {parent=upperarm_twist_01_l, loc=<-11.2012 -1.2234 8.7784>, rot=<0 0 0 -1>}
 deform_upperarm_up_l {parent=deform_upperarm_up_base_l, loc=<0 0 0>, rot=<0 0 0 1>}
 upperarm_twist_02_l {parent=upperarm_l, loc=<19.5059 0.0843 -0.6195>, rot=<0 0 0 -1>}
 deform_bicep_base_l {parent=upperarm_twist_02_l, loc=<0 8.0494 0>, rot=<0 0 0 1>}
 deform_bicep_l {parent=deform_bicep_base_l, loc=<0 0 0>, rot=<0 0 0 1>}
 base_corrective_FurCollar_l {parent=spine_05, loc=<15.0162 0.7128 -10.822>, rot=<0.0284 -0.7065 -0.0284 -0.7065>}
 corrective_FurCollar_l {parent=base_corrective_FurCollar_l, loc=<0 0 0>, rot=<0 0 0 1>}
 base_furCollarPiece_l_03 {parent=corrective_FurCollar_l, loc=<-0.2379 0.0263 3.058>, rot=<0.3053 0.2012 0.8763 -0.3135>}
 dyn_furCollarPiece_l_03 {parent=base_furCollarPiece_l_03, loc=<0 0 0>, rot=<0 0 0 1>}
 attach_cape {parent=spine_05, loc=<9.1509 -13.3783 0>, rot=<0.0616 -0.7044 -0.0616 -0.7044>}
 base_corrective_FurCollar_b {parent=spine_05, loc=<19.7306 -5.8544 0>, rot=<0.5196 -0.4795 0.4795 -0.5196>}
 corrective_FurCollar_b {parent=base_corrective_FurCollar_b, loc=<0 0 0>, rot=<0 0 0 1>}
 base_furCollarPiece_bl_01 {parent=corrective_FurCollar_b, loc=<1.8608 4.0558 1.3906>, rot=<-0.037 0.1322 -0.2672 -0.9538>}
 dyn_furCollarPiece_bl_01 {parent=base_furCollarPiece_bl_01, loc=<0 0 0>, rot=<0 0 0 1>}
 base_furCollarPiece_b_01 {parent=corrective_FurCollar_b, loc=<1.1361 0.9042 2.4106>, rot=<-0.0056 0.0308 -0.0794 -0.9964>}
 dyn_furCollarPiece_b_01 {parent=base_furCollarPiece_b_01, loc=<0 0 0>, rot=<0 0 0 1>}
 base_furCollarPiece_b_02 {parent=corrective_FurCollar_b, loc=<6.6754 -0.4787 -1.337>, rot=<-0.1456 0.0646 0.0456 -0.9862>}
 dyn_furCollarPiece_b_02 {parent=base_furCollarPiece_b_02, loc=<0 0 0>, rot=<0 0 0 1>}
 base_furCollarPiece_br_01 {parent=corrective_FurCollar_b, loc=<1.5334 -2.902 2.3063>, rot=<-0.069 0.066 0.1277 -0.9872>}
 dyn_furCollarPiece_br_01 {parent=base_furCollarPiece_br_01, loc=<0 0 0>, rot=<0 0 0 1>}
 base_furCollarPiece_br_02 {parent=corrective_FurCollar_b, loc=<1.7733 -7.3569 1.9315>, rot=<-0.1128 -0.0528 0.1565 -0.9798>}
 dyn_furCollarPiece_br_02 {parent=base_furCollarPiece_br_02, loc=<0 0 0>, rot=<0 0 0 1>}
 base_furCollarPiece_br_03 {parent=corrective_FurCollar_b, loc=<5.0289 -7.1681 -0.5423>, rot=<-0.0973 0.011 0.0862 -0.9915>}
 dyn_furCollarPiece_br_03 {parent=base_furCollarPiece_br_03, loc=<0 0 0>, rot=<0 0 0 1>}
 base_furCollarPiece_br_04 {parent=corrective_FurCollar_b, loc=<2.3509 -10.501 -0.2651>, rot=<-0.1906 0.075 0.0455 -0.9777>}
 dyn_furCollarPiece_br_04 {parent=base_furCollarPiece_br_04, loc=<0 0 0>, rot=<0 0 0 1>}
 base_furCollarPiece_l_01 {parent=spine_05, loc=<10.9644 9.504 -14.6087>, rot=<0.6257 0.1331 -0.2011 -0.7418>}
 dyn_furCollarPiece_l_01 {parent=base_furCollarPiece_l_01, loc=<0 0 0>, rot=<0 0 0 1>}
 base_furCollarPiece_l_02 {parent=spine_05, loc=<8.352 7.7666 -18.3303>, rot=<0.7968 -0.3393 -0.2874 -0.4091>}
 dyn_furCollarPiece_l_02 {parent=base_furCollarPiece_l_02, loc=<0 0 0>, rot=<0 0 0 1>}
 base_furCollarPiece_l_04 {parent=spine_05, loc=<11.3404 -5.1551 -21.4659>, rot=<-0.0021 -0.6505 0.207 -0.7307>}
 dyn_furCollarPiece_l_04 {parent=base_furCollarPiece_l_04, loc=<0 0 0>, rot=<0 0 0 1>}
 base_furCollarPiece_l_05 {parent=spine_05, loc=<18.3324 -5.8928 -16.9596>, rot=<0.8568 -0.2932 0.2615 -0.334>}
 dyn_furCollarPiece_l_05 {parent=base_furCollarPiece_l_05, loc=<0 0 0>, rot=<0 0 0 1>}
 base_furCollarPiece_bl_02 {parent=spine_05, loc=<8.4681 -17.1211 -10.6644>, rot=<0.3248 -0.6888 0.1828 -0.6218>}
 dyn_furCollarPiece_bl_02 {parent=base_furCollarPiece_bl_02, loc=<0 0 0>, rot=<0 0 0 1>}
 base_furCollarPiece_bl_03 {parent=spine_05, loc=<6.7272 -12.7079 -17.0801>, rot=<0.2407 -0.7316 0.0706 -0.6339>}
 dyn_furCollarPiece_bl_03 {parent=base_furCollarPiece_bl_03, loc=<0 0 0>, rot=<0 0 0 1>}
 base_furCollarPiece_bl_04 {parent=spine_05, loc=<16.6966 -10.1395 -13.7429>, rot=<0.3724 -0.7421 0.4323 -0.3518>}
 dyn_furCollarPiece_bl_04 {parent=base_furCollarPiece_bl_04, loc=<0 0 0>, rot=<0 0 0 1>}
 base_furCollarPiece_b_03 {parent=spine_05, loc=<17.9008 -15.2718 -4.0411>, rot=<0.3784 -0.621 0.5268 -0.44>}
 dyn_furCollarPiece_b_03 {parent=base_furCollarPiece_b_03, loc=<0 0 0>, rot=<0 0 0 1>}
 base_furCollarPiece_b_04 {parent=spine_05, loc=<11.6385 -18.1295 -3.1196>, rot=<0.1232 -0.664 0.2975 -0.6748>}
 dyn_furCollarPiece_b_04 {parent=base_furCollarPiece_b_04, loc=<0 0 0>, rot=<0 0 0 1>}
 base_furCollarPiece_br_05 {parent=spine_05, loc=<11.5336 -16.6506 7.8319>, rot=<0.4329 -0.4796 0.3684 -0.6685>}
 dyn_furCollarPiece_br_05 {parent=base_furCollarPiece_br_05, loc=<0 0 0>, rot=<0 0 0 1>}
 base_furCollarPiece_br_06 {parent=spine_05, loc=<8.7125 -19.4531 4.3156>, rot=<0.4898 -0.4192 0.563 -0.517>}
 dyn_furCollarPiece_br_06 {parent=base_furCollarPiece_br_06, loc=<0 0 0>, rot=<0 0 0 1>}
 clavicle_r {parent=spine_05, loc=<2.5855 -0.2866 4.3777>, rot=<0.7289 0.0465 -0.6829 0.009>}
 upperarm_r {parent=clavicle_r, loc=<-18.2264 -0.1342 0.001>, rot=<0.0253 -0.2693 0.0013 -0.9627>}
 lowerarm_r {parent=upperarm_r, loc=<-29.0071 -0.1686 1.239>, rot=<0.0391 -0.0099 -0.2453 -0.9686>}
 hand_r {parent=lowerarm_r, loc=<-25.8607 -0.222 -1.0274>, rot=<-0.7193 -0.0988 0.1193 -0.6773>}
 thumb_01_r {parent=hand_r, loc=<-1.3118 0.3191 2.459>, rot=<0.6546 -0.4042 0.0664 -0.6354>}
 thumb_02_r {parent=thumb_01_r, loc=<-5.9336 0.0448 -0.0401>, rot=<-0.0037 0.0213 0.298 -0.9543>}
 thumb_03_r {parent=thumb_02_r, loc=<-2.2223 -0.3682 0.0907>, rot=<0.0118 -0.013 0.0066 -0.9998>}
 index_metacarpal_r {parent=hand_r, loc=<-1.6535 -0.0183 2.96>, rot=<0.0158 -0.0124 -0.0173 -0.9996>}
 index_01_r {parent=index_metacarpal_r, loc=<-8.3008 -0.0357 0.0524>, rot=<-0.0044 0.0485 0.0924 -0.9945>}
 index_02_r {parent=index_01_r, loc=<-3.3587 0.3004 -0.0452>, rot=<-0.002 0.0027 0.3225 -0.9466>}
 index_03_r {parent=index_02_r, loc=<-2.5578 0.123 0.0252>, rot=<-0.002 0.0067 0.2558 -0.9667>}
 middle_metacarpal_r {parent=hand_r, loc=<-2.1856 0.1725 0.5984>, rot=<0.006 0.0168 -0.0329 -0.9993>}
 middle_01_r {parent=middle_metacarpal_r, loc=<-7.3275 0.2447 -0.0129>, rot=<-0.0208 0.0042 0.0858 -0.9961>}
 middle_02_r {parent=middle_01_r, loc=<-4.0123 -0.1401 -0.1175>, rot=<-0.0218 0.0258 0.3512 -0.9357>}
 middle_03_r {parent=middle_02_r, loc=<-2.6077 -0.0534 -0.1373>, rot=<0.0082 0.0188 0.1708 -0.9851>}
 ring_metacarpal_r {parent=hand_r, loc=<-2.2074 0.3625 -1.3638>, rot=<-0.0284 0.0722 -0.0458 -0.9959>}
 ring_01_r {parent=ring_metacarpal_r, loc=<-7.0295 0.6197 0.1446>, rot=<-0.0489 -0.0252 0.1892 -0.9804>}
 ring_02_r {parent=ring_01_r, loc=<-3.6174 -0.4647 0.1027>, rot=<0 0 0.2806 -0.9598>}
 ring_03_r {parent=ring_02_r, loc=<-2.4152 -0.114 -0.0199>, rot=<0 0 0.1189 -0.9929>}
 pinky_metacarpal_r {parent=hand_r, loc=<-1.7616 0.9753 -3.3814>, rot=<-0.0397 0.0814 -0.069 -0.9935>}
 pinky_01_r {parent=pinky_metacarpal_r, loc=<-7.1709 0.4831 0.1094>, rot=<-0.0654 -0.0035 0.2242 -0.9723>}
 pinky_02_r {parent=pinky_01_r, loc=<-2.8274 -0.0893 0.1043>, rot=<0.0037 0.0011 0.2854 -0.9584>}
 pinky_03_r {parent=pinky_02_r, loc=<-2.2046 -0.0307 0.0078>, rot=<-0.0002 0.0098 0.1445 -0.9895>}
 deform_wrist_dn_base_r {parent=hand_r, loc=<0 3.4172 0>, rot=<0 0 0 -1>}
 deform_wrist_dn_r {parent=deform_wrist_dn_base_r, loc=<0 0 0>, rot=<0 0 0 -1>}
 deform_wrist_up_base_r {parent=hand_r, loc=<-0.0001 -2.7804 0>, rot=<0 0 0 -1>}
 deform_wrist_up_r {parent=deform_wrist_up_base_r, loc=<0 0 0>, rot=<0 0 0 -1>}
 lowerarm_twist_02_r {parent=lowerarm_r, loc=<-9.6199 0 0>, rot=<0 0 0 -1>}
 lowerarm_twist_01_r {parent=lowerarm_r, loc=<-17.8549 0 0>, rot=<0 0 0 -1>}
 deform_elbow_in_base_r {parent=lowerarm_r, loc=<-0.2114 -5.6449 -0.827>, rot=<0 0 0 -1>}
 deform_elbow_in_r {parent=deform_elbow_in_base_r, loc=<0 0 0>, rot=<0 0 0 1>}
 deform_elbow_out_base_r {parent=lowerarm_r, loc=<-0.0001 7.5129 0.0002>, rot=<0 0 0 -1>}
 deform_elbow_out_r {parent=deform_elbow_out_base_r, loc=<0 0 0>, rot=<0 0 0 1>}
 upperarm_twist_01_r {parent=upperarm_r, loc=<-7.9796 -0.0843 0.6195>, rot=<0 0 0 -1>}
 deform_upperarm_bk_base_r {parent=upperarm_twist_01_r, loc=<6.8985 13.3826 1.3257>, rot=<0 0 0 1>}
 deform_upperarm_bk_r {parent=deform_upperarm_bk_base_r, loc=<0 0 0>, rot=<0 0 0 1>}
 deform_upperarm_fr_base_r {parent=upperarm_twist_01_r, loc=<8.3485 -10.7117 2.1456>, rot=<0 0 0 1>}
 deform_upperarm_fr_r {parent=deform_upperarm_fr_base_r, loc=<0 0 0>, rot=<0 0 0 1>}
 deform_upperarm_up_base_r {parent=upperarm_twist_01_r, loc=<11.2015 1.2234 -8.7788>, rot=<0 0 0 1>}
 deform_upperarm_up_r {parent=deform_upperarm_up_base_r, loc=<0 0 0>, rot=<0 0 0 1>}
 upperarm_twist_02_r {parent=upperarm_r, loc=<-19.5059 -0.0843 0.6195>, rot=<0 0 0 -1>}
 deform_bicep_base_r {parent=upperarm_twist_02_r, loc=<0.0002 -8.0494 -0.0003>, rot=<0 0 0 1>}
 deform_bicep_r {parent=deform_bicep_base_r, loc=<0 0 0>, rot=<0 0 0 1>}
 base_corrective_FurCollar_r {parent=spine_05, loc=<15.016 0.7128 10.822>, rot=<0.7065 0.0284 0.7065 -0.0284>}
 corrective_FurCollar_r {parent=base_corrective_FurCollar_r, loc=<0 0 0>, rot=<0 0 0 1>}
 base_furCollarPiece_r_03 {parent=corrective_FurCollar_r, loc=<1.6479 -4.368 2.2938>, rot=<0.356 -0.7605 0.3672 0.4>}
 dyn_furCollarPiece_r_03 {parent=base_furCollarPiece_r_03, loc=<0 0 0>, rot=<0 0 0 1>}
 base_furCollarPiece_r_04 {parent=corrective_FurCollar_r, loc=<8.3447 -3.6917 -0.0406>, rot=<-0.5513 0.243 -0.4616 -0.6511>}
 dyn_furCollarPiece_r_04 {parent=base_furCollarPiece_r_04, loc=<0 0 0>, rot=<0 0 0 1>}
 base_furCollarPiece_r_05 {parent=corrective_FurCollar_r, loc=<1.2431 0.5157 3.8325>, rot=<-0.1469 0.1371 -0.4015 -0.8935>}
 dyn_furCollarPiece_r_05 {parent=base_furCollarPiece_r_05, loc=<0 0 0>, rot=<0 0 0 1>}
 base_furCollarPiece_r_06 {parent=corrective_FurCollar_r, loc=<7.1942 -0.7714 2.4062>, rot=<-0.6581 0.1147 -0.2947 -0.6833>}
 dyn_furCollarPiece_r_06 {parent=base_furCollarPiece_r_06, loc=<0 0 0>, rot=<0 0 0 1>}
 base_furCollarPiece_r_07 {parent=corrective_FurCollar_r, loc=<0.7391 5.6 4.84>, rot=<-0.1554 0.2648 -0.6574 -0.6882>}
 dyn_furCollarPiece_r_07 {parent=base_furCollarPiece_r_07, loc=<0 0 0>, rot=<0 0 0 1>}
 base_furCollarPiece_r_08 {parent=corrective_FurCollar_r, loc=<2.6604 7.3022 3.1152>, rot=<-0.2327 -0.2174 -0.3785 -0.8691>}
 dyn_furCollarPiece_r_08 {parent=base_furCollarPiece_r_08, loc=<0 0 0>, rot=<0 0 0 1>}
 base_furCollarPiece_r_01 {parent=spine_05, loc=<8.3422 9.3361 10.1657>, rot=<0.7159 0.1985 -0.0191 -0.6691>}
 dyn_furCollarPiece_r_01 {parent=base_furCollarPiece_r_01, loc=<0 0 0>, rot=<0 0 0 1>}
 base_furCollarPiece_r_02 {parent=spine_05, loc=<9.7079 4.1452 19.8983>, rot=<0.3234 0.548 0.4764 -0.6068>}
 dyn_furCollarPiece_r_02 {parent=base_furCollarPiece_r_02, loc=<0 0 0>, rot=<0 0 0 1>}
 base_furCollarPiece_r_09 {parent=spine_05, loc=<12.5281 -5.3526 21.7467>, rot=<0.5826 -0.3634 0.4518 -0.5696>}
 dyn_furCollarPiece_r_09 {parent=base_furCollarPiece_r_09, loc=<0 0 0>, rot=<0 0 0 1>}
 base_furCollarPiece_r_10 {parent=spine_05, loc=<13.5184 -8.4287 18.9672>, rot=<0.7251 -0.3478 0.4856 -0.3427>}
 dyn_furCollarPiece_r_10 {parent=base_furCollarPiece_r_10, loc=<0 0 0>, rot=<0 0 0 1>}
 deform_pec_aim_base_l {parent=spine_04, loc=<5.0232 13.6719 -2.6508>, rot=<0.0672 -0.5201 0.3228 -0.7879>}
 deform_pec_aim_l {parent=deform_pec_aim_base_l, loc=<0 0 0>, rot=<0 0 0 1>}
 deform_pec_base_l {parent=deform_pec_aim_l, loc=<16.3279 0 0>, rot=<0 0 0 1>}
 deform_pec_l {parent=deform_pec_base_l, loc=<0 0 0>, rot=<0 0 0 1>}
 deform_pec_aim_base_r {parent=spine_04, loc=<5.023 13.6718 2.6508>, rot=<-0.5201 -0.0672 0.7879 0.3228>}
 deform_pec_aim_r {parent=deform_pec_aim_base_r, loc=<0 0 0>, rot=<0 0 0 1>}
 deform_pec_base_r {parent=deform_pec_aim_r, loc=<-16.3279 0 0>, rot=<0 0 0 -1>}
 deform_pec_r {parent=deform_pec_base_r, loc=<0 0 0.0002>, rot=<0 0 0 -1>}
 deform_lat_l_base {parent=spine_03, loc=<2.5566 -10.3716 -17.6918>, rot=<0 0 0 1>}
 deform_lat_l {parent=deform_lat_l_base, loc=<0 0 0>, rot=<0 0 0 1>}
 deform_lat_r_base {parent=spine_03, loc=<2.5568 -10.3716 17.6918>, rot=<0 0 1 0>}
 deform_lat_r {parent=deform_lat_r_base, loc=<0 0 0>, rot=<0 0 0 -1>}
 thigh_l {parent=pelvis, loc=<-1.8716 -0.0109 -11.9936>, rot=<0.0012 0.0572 0.0155 -0.9982>}
 calf_l {parent=thigh_l, loc=<-41.9178 -0.2246 0.0926>, rot=<0.0006 -0.0161 -0.0391 -0.9991>}
 foot_l {parent=calf_l, loc=<-38.7057 1.2438 0.2452>, rot=<0.0975 -0.0474 0.0317 -0.9936>}
 ball_l {parent=foot_l, loc=<-6.1614 13.1469 0.6807>, rot=<0.0001 -0.0001 -0.7063 -0.7079>}
 calf_twist_02_l {parent=calf_l, loc=<-15.9268 -0.0967 0.0397>, rot=<0.0237 0.0095 -0.0035 -0.9997>}
 base_bootFlap_l {parent=calf_twist_02_l, loc=<0.4669 -11.613 -0.7196>, rot=<0.6724 -0.2173 0.1382 -0.6939>}
 dyn_bootFlap_l {parent=base_bootFlap_l, loc=<0 0 0>, rot=<0 0 0 1>}
 calf_twist_01_l {parent=calf_l, loc=<-32.9335 -0.0967 0.0397>, rot=<0.0237 0.0095 -0.0035 -0.9997>}
 deform_knee_in_base_l {parent=calf_l, loc=<1.298 -8.0356 -0.0496>, rot=<0 0 0 1>}
 deform_knee_in_l {parent=deform_knee_in_base_l, loc=<0 0 0>, rot=<0 0 0 1>}
 deform_knee_out_base_l {parent=calf_l, loc=<0 6.8739 0>, rot=<0 0 0 -1>}
 deform_knee_out_l {parent=deform_knee_out_base_l, loc=<0 0 0>, rot=<0 0 0 1>}
 thigh_twist_01_l {parent=thigh_l, loc=<-21.6303 0 0>, rot=<0.0013 0 0 -1>}
 deform_groin_base_l {parent=thigh_l, loc=<-2.1908 8.5661 -1.7442>, rot=<0 0 0 -1>}
 deform_groin_l {parent=deform_groin_base_l, loc=<0 0 0>, rot=<0 0 0 1>}
 thigh_r {parent=pelvis, loc=<-1.8716 -0.0109 11.9936>, rot=<0.0572 -0.0012 0.9982 0.0155>}
 calf_r {parent=thigh_r, loc=<41.9178 0.2246 -0.0926>, rot=<0.0006 -0.0161 -0.0391 -0.9991>}
 foot_r {parent=calf_r, loc=<38.7057 -1.2438 -0.2452>, rot=<0.0975 -0.0474 0.0317 -0.9936>}
 ball_r {parent=foot_r, loc=<6.1614 -13.1469 -0.6807>, rot=<0.0001 -0.0001 -0.7063 -0.7079>}
 calf_twist_02_r {parent=calf_r, loc=<15.9268 0.0967 -0.0397>, rot=<0.0237 0.0095 -0.0035 -0.9997>}
 base_bootFlap_r {parent=calf_twist_02_r, loc=<-0.4669 11.613 0.7196>, rot=<-0.1382 -0.6939 0.6724 0.2173>}
 dyn_bootFlap_r {parent=base_bootFlap_r, loc=<0 0 0>, rot=<0 0 0 1>}
 calf_twist_01_r {parent=calf_r, loc=<32.9335 0.0967 -0.0397>, rot=<0.0237 0.0095 -0.0035 -0.9997>}
 deform_knee_in_base_r {parent=calf_r, loc=<-1.2981 8.0356 0.0496>, rot=<0 0 0 -1>}
 deform_knee_in_r {parent=deform_knee_in_base_r, loc=<0 0 0>, rot=<0 0 0 1>}
 deform_knee_out_base_r {parent=calf_r, loc=<0 -6.8739 0>, rot=<0 0 0 -1>}
 deform_knee_out_r {parent=deform_knee_out_base_r, loc=<0 0 0>, rot=<0 0 0 1>}
 thigh_twist_01_r {parent=thigh_r, loc=<21.6303 0 0>, rot=<0.0013 0 0 -1>}
 deform_groin_base_r {parent=thigh_r, loc=<2.1908 -8.5661 1.7442>, rot=<0 0 0 -1>}
 deform_groin_r {parent=deform_groin_base_r, loc=<0 0 0>, rot=<0 0 0 -1>}
 deform_glute_base_l {parent=pelvis, loc=<-10.1877 -11.6217 -12.899>, rot=<0.0275 0.0501 -0.4472 -0.8926>}
 deform_glute_l {parent=deform_glute_base_l, loc=<0 0 0>, rot=<0 0 0 1>}
 deform_glute_base_r {parent=pelvis, loc=<-10.1877 -11.6218 12.899>, rot=<0.0501 -0.0275 0.8926 -0.4472>}
 deform_glute_r {parent=deform_glute_base_r, loc=<0 0 0>, rot=<0 0 0 1>}
]
variables [
 Knee_in {type=FTransform, default=(Rotation=(X=0,Y=0,Z=0,W=1),Translation=(X=0,Y=-12,Z=0),Scale3D=(X=0.5,Y=1,Z=1))}
 Knee_out {type=FTransform, default=(Rotation=(X=0,Y=0,Z=0,W=1),Translation=(X=0,Y=3,Z=0),Scale3D=(X=1.5,Y=1,Z=1.5))}
]
