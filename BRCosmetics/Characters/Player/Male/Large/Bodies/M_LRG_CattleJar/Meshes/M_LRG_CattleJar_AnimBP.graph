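# Animation Blueprint for the CattleJar large male player body, adding secondary motion and correctives to fur collar, belt buckle and boot flaps.
# It copies the attached parent's pose, then chains Control Rig, weighted bone constraints and per-piece spring AnimDynamics through cached poses.
# Additive bone transforms finish the pose, some always on, others weighted by knee bend, head tilt or DBNO state.
# The EventGraph derives those alphas from reference-pose deltas, suppressing knee correctives during emotes.
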AnimBlueprint /BRCosmetics/Characters/Player/Male/Large/Bodies/M_LRG_CattleJar/Meshes/M_LRG_CattleJar_AnimBP
target_skeleton=&/M_LRG_CattleJar_Skeleton
graphs [
 AnimGraph {nodes [
   n1 = "Transform (Modify) Bone - Bone: corrective_FurCollar_b"(
    ComponentPose=Local To Component(LocalPose=Use cached pose 'AnimDyn2Logic'())
    Alpha=Get Head Up() -> (Head Up:float)
    Translation="11.915,0,-4.7"
    Rotation="-36.712,-0.001,-0.070"
    Scale="1,1,1"
   ) {
    rotationMode=BMM_Additive
    rotationSpace=BCS_BoneSpace
    translationMode=BMM_Additive
    translationSpace=BCS_BoneSpace
    alphaBoolBlend {blendOption=Linear}
    alphaScaleBiasClamp {bInterpResult=true}
    boneToModify {boneName=corrective_FurCollar_b}
    rotation=<rot -36.712 -0.001 -0.07>
    translation {x=11.915, z=-4.7}
   }
   n2 = "Transform (Modify) Bone - Bone: base_bootFlap_l"(
    ComponentPose=@n1
    Alpha=Get Calf L Rot() -> (Calf L Rot:float)
    Translation="-10.467,0.187,-3.047"
    Rotation="-82.905,-15.965,21.532"
    Scale="0.436,0.800,0.550"
   ) {
    rotationMode=BMM_Additive
    rotationSpace=BCS_BoneSpace
    scaleMode=BMM_Additive
    scaleSpace=BCS_BoneSpace
    translationMode=BMM_Additive
    translationSpace=BCS_BoneSpace
    alphaBoolBlend {blendOption=Linear}
    alphaScaleBiasClamp {bInterpResult=true}
    boneToModify {boneName=base_bootFlap_l}
    rotation=<rot -82.905 -15.965 21.532>
    scale=<0.436 0.8 0.5503>
    translation=<-10.467 0.187 -3.047>
   }
   n3 = "Transform (Modify) Bone - Bone: base_bootFlap_r"(
    ComponentPose=@n2
    Alpha=Get Calf R Rot() -> (Calf R Rot:float)
    Translation="-13.116,-0.100,3.340"
    Rotation="-64.834,-163.350,171.332"
    Scale="0.7,0.7,0.7"
   ) {
    rotationMode=BMM_Additive
    rotationSpace=BCS_BoneSpace
    scaleMode=BMM_Additive
    scaleSpace=BCS_BoneSpace
    translationMode=BMM_Additive
    translationSpace=BCS_BoneSpace
    alphaBoolBlend {blendOption=Linear}
    alphaScaleBiasClamp {bInterpResult=true}
    boneToModify {boneName=base_bootFlap_r}
    rotation=<rot -64.834 -163.35 171.332>
    scale=<0.7 0.7 0.7>
    translation=<-13.116 -0.1 3.34>
   }
   n4 = "Transform (Modify) Bone - Bone: base_bootFlap_l"(
    ComponentPose=@n3
    Alpha="Get Is DBNO "() -> ("Is DBNO ":double)
    Translation="-0.744,3.219,-12.485"
    Rotation="22.509,-8.070,-26.295"
    Scale="0.7,0.7,0.7"
   ) {
    rotationMode=BMM_Additive
    rotationSpace=BCS_BoneSpace
    scaleMode=BMM_Additive
    scaleSpace=BCS_BoneSpace
    translationMode=BMM_Additive
    translationSpace=BCS_BoneSpace
    alphaBoolBlend {blendOption=Linear}
    alphaScaleBiasClamp {bInterpResult=true}
    boneToModify {boneName=base_bootFlap_l}
    rotation=<rot 22.5089 -8.0699 -26.295>
    scale=<0.7 0.7 0.7>
    translation=<-0.744 3.219 -12.485>
   }
   n5 = "Transform (Modify) Bone - Bone: Corrective_Belt_Buckle"!(ComponentPose=@n4, bAlphaBoolEnabled:bool=True, Translation="3.563,-0.295,0.008", Rotation="3.410,-0.026,-0.414", Scale="0.9,0.9,0.9") -> (Pose) {
    alphaInputType=Bool
    rotationMode=BMM_Additive
    rotationSpace=BCS_BoneSpace
    scaleMode=BMM_Additive
    scaleSpace=BCS_BoneSpace
    translationMode=BMM_Additive
    translationSpace=BCS_BoneSpace
    alphaBoolBlend {blendOption=Linear}
    boneToModify {boneName=Corrective_Belt_Buckle}
    rotation=<rot 3.4101 -0.026 -0.414>
    scale=<0.9 0.9 0.9>
    translation=<3.563 -0.295 0.008>
   }
   n6 = "Transform (Modify) Bone - Bone: base_corrective_FurCollar_l"!(ComponentPose=@n5, bAlphaBoolEnabled:bool=True, Translation="2.079,-0.077,0.714", Rotation="9.314,-0.003,-0.001", Scale="1,1,0.7") -> (Pose) {
    alphaInputType=Bool
    rotationMode=BMM_Additive
    rotationSpace=BCS_BoneSpace
    scaleMode=BMM_Additive
    scaleSpace=BCS_BoneSpace
    translationMode=BMM_Additive
    translationSpace=BCS_ParentBoneSpace
    alphaBoolBlend {blendOption=Linear}
    boneToModify {boneName=base_corrective_FurCollar_l}
    rotation=<rot 9.314 -0.003 -0.001>
    scale {z=0.7}
    translation=<2.079 -0.077 0.714>
   }
   n7 = "Transform (Modify) Bone - Bone: base_furCollarPiece_l_05"!(ComponentPose=@n6, bAlphaBoolEnabled:bool=True, Translation="0,-0.6,0", Rotation="-0.183,66.537,-0.097", Scale="1,1,1") -> (Pose) {
    alphaInputType=Bool
    rotationMode=BMM_Additive
    rotationSpace=BCS_BoneSpace
    scaleMode=BMM_Additive
    scaleSpace=BCS_BoneSpace
    translationMode=BMM_Additive
    translationSpace=BCS_ParentBoneSpace
    alphaBoolBlend {blendOption=Linear}
    boneToModify {boneName=base_furCollarPiece_l_05}
    rotation=<rot -0.183 66.537 -0.097>
    translation {y=-0.6}
   }
   n8 = "Transform (Modify) Bone - Bone: base_furCollarPiece_l_04"!(ComponentPose=@n7, bAlphaBoolEnabled:bool=True, Translation="0,0,1.635", Rotation="0,0,0", Scale="1,1,1") -> (Pose) {
    alphaInputType=Bool
    rotationMode=BMM_Additive
    rotationSpace=BCS_BoneSpace
    scaleMode=BMM_Additive
    scaleSpace=BCS_BoneSpace
    translationMode=BMM_Additive
    translationSpace=BCS_ParentBoneSpace
    alphaBoolBlend {blendOption=Linear}
    boneToModify {boneName=base_furCollarPiece_l_04}
    translation {z=1.635}
   }
   n9 = "Transform (Modify) Bone - Bone: base_corrective_FurCollar_r"!(ComponentPose=@n8, bAlphaBoolEnabled:bool=True, Translation="6.177,-0.969,3.531", Rotation="1.610,-37.420,37.655", Scale="0.500,0.500,0.500") -> (Pose) {
    alphaInputType=Bool
    rotationMode=BMM_Additive
    rotationSpace=BCS_BoneSpace
    scaleMode=BMM_Additive
    scaleSpace=BCS_BoneSpace
    translationMode=BMM_Additive
    translationSpace=BCS_ParentBoneSpace
    alphaBoolBlend {blendOption=Linear}
    boneToModify {boneName=base_corrective_FurCollar_r}
    rotation=<rot 1.61 -37.42 37.655>
    scale=<0.5 0.5 0.5>
    translation=<6.177 -0.969 3.531>
   }
   n10 = "Transform (Modify) Bone - Bone: base_furCollarPiece_r_09"!(ComponentPose=@n9, bAlphaBoolEnabled:bool=True, Translation="-6.056,0.058,-6.032", Rotation="0,0,0", Scale="0.5,0.5,0.5") -> (Pose) {
    alphaInputType=Bool
    rotationMode=BMM_Additive
    rotationSpace=BCS_BoneSpace
    scaleMode=BMM_Additive
    scaleSpace=BCS_BoneSpace
    translationMode=BMM_Additive
    translationSpace=BCS_ParentBoneSpace
    alphaBoolBlend {blendOption=Linear}
    boneToModify {boneName=base_furCollarPiece_r_09}
    scale=<0.5 0.5 0.5>
    translation=<-6.056 0.058 -6.032>
   }
   Output Pose(Result=Component To Local(ComponentPose=@n10)) {layerGroup=None}
   n11 = Control Rig(Source=Copy Pose From Mesh() {bUseAttachedParent=true}) {
    alphaBoolBlend {blendOption=Linear}
    controlRigAssetReference {blueprintRigClass=&/M_LRG_CattleJar_CtrlRig.M_LRG_CattleJar_CtrlRig_C}
   }
   Save cached pose 'ConRig2Constrains'(Pose=@n11) {cacheName=ConRig2Constrains}
   n12 = "Constraint - base_corrective_FurCollar_l (PPP)"!(
    ComponentPose=Local To Component(LocalPose=Use cached pose 'ConRig2Constrains'())
    Alpha=1
    ConstraintWeights_0=0.6
    ConstraintWeights_1:float=0.3
    ConstraintWeights_2:float=0.1
   ) -> (Pose) {
    alphaBoolBlend {blendOption=Linear}
    boneToModify {boneName=base_corrective_FurCollar_l}
    constraintSetup [{targetBone {boneName=spine_05}}, {targetBone {boneName=clavicle_l}}, {targetBone {boneName=deform_upperarm_up_l}}]
    constraintWeights [1,1,1]
   }
   n13 = "Constraint - base_furCollarPiece_l_01 (PP)"!(ComponentPose=@n12, Alpha=1, ConstraintWeights_0=0.7, ConstraintWeights_1:float=0.3) -> (Pose) {
    alphaBoolBlend {blendOption=Linear}
    boneToModify {boneName=base_furCollarPiece_l_01}
    constraintSetup [{targetBone {boneName=spine_05}}, {targetBone {boneName=clavicle_l}}]
    constraintWeights [1,1]
   }
   n14 = "Constraint - base_furCollarPiece_l_02 (PPP)"!(ComponentPose=@n13, Alpha=1, ConstraintWeights_0=0.2, ConstraintWeights_1:float=0.5, ConstraintWeights_2:float=0.3) -> (Pose) {
    alphaBoolBlend {blendOption=Linear}
    boneToModify {boneName=base_furCollarPiece_l_02}
    constraintSetup [{targetBone {boneName=deform_upperarm_up_l}}, {targetBone {boneName=clavicle_l}}, {targetBone {boneName=spine_05}}]
    constraintWeights [1,1,1]
   }
   n15 = "Constraint - base_furCollarPiece_l_04 (PPP)"!(ComponentPose=@n14, Alpha=1, ConstraintWeights_0=0.4, ConstraintWeights_1:float=0.3, ConstraintWeights_2:float=0.3) -> (Pose) {
    alphaBoolBlend {blendOption=Linear}
    boneToModify {boneName=base_furCollarPiece_l_04}
    constraintSetup [{targetBone {boneName=clavicle_l}}, {targetBone {boneName=deform_upperarm_up_l}}, {targetBone {boneName=spine_05}}]
    constraintWeights [1,1,1]
   }
   n16 = "Constraint - base_furCollarPiece_bl_02 (PP)"!(ComponentPose=@n15, Alpha=1, ConstraintWeights_0=0.4, ConstraintWeights_1:float=0.6) -> (Pose) {
    alphaBoolBlend {blendOption=Linear}
    boneToModify {boneName=base_furCollarPiece_bl_02}
    constraintSetup [{targetBone {boneName=clavicle_l}}, {targetBone {boneName=spine_05}}]
    constraintWeights [1,1]
   }
   n17 = "Constraint - base_furCollarPiece_bl_03 (PPP)"!(ComponentPose=@n16, Alpha=1, ConstraintWeights_0=0.5, ConstraintWeights_1:float=0.2, ConstraintWeights_2:float=0.3) -> (Pose) {
    alphaBoolBlend {blendOption=Linear}
    boneToModify {boneName=base_furCollarPiece_bl_03}
    constraintSetup [{targetBone {boneName=clavicle_l}}, {targetBone {boneName=deform_upperarm_bk_l}}, {targetBone {boneName=spine_05}}]
    constraintWeights [1,1,1]
   }
   n18 = "Constraint - base_furCollarPiece_bl_04 (PP)"!(ComponentPose=@n17, Alpha=1, ConstraintWeights_0=0.7, ConstraintWeights_1:float=0.3) -> (Pose) {
    alphaBoolBlend {blendOption=Linear}
    boneToModify {boneName=base_furCollarPiece_bl_04}
    constraintSetup [{targetBone {boneName=clavicle_l}}, {targetBone {boneName=deform_upperarm_bk_l}}]
    constraintWeights [1,1]
   }
   n19 = "Constraint - base_furCollarPiece_l_05 (PPP)"!(ComponentPose=@n18, Alpha=1, ConstraintWeights_0=0.5, ConstraintWeights_1:float=0.25, ConstraintWeights_2:float=0.25) -> (Pose) {
    alphaBoolBlend {blendOption=Linear}
    boneToModify {boneName=base_furCollarPiece_l_05}
    constraintSetup [{targetBone {boneName=clavicle_l}}, {targetBone {boneName=spine_05}}, {targetBone {boneName=base_furCollarPiece_l_04}}]
    constraintWeights [1,1,1]
   }
   n20 = "Constraint - base_corrective_FurCollar_r (PPP)"!(ComponentPose=@n19, Alpha=1, ConstraintWeights_0=0.55, ConstraintWeights_1:float=0.3, ConstraintWeights_2:float=0.15) -> (Pose) {
    alphaBoolBlend {blendOption=Linear}
    boneToModify {boneName=base_corrective_FurCollar_r}
    constraintSetup [{targetBone {boneName=spine_05}}, {targetBone {boneName=clavicle_r}}, {targetBone {boneName=deform_upperarm_up_r}}]
    constraintWeights [1,1,1]
   }
   n21 = "Constraint - base_furCollarPiece_r_01 (PP)"!(ComponentPose=@n20, Alpha=1, ConstraintWeights_0=0.7, ConstraintWeights_1:float=0.3) -> (Pose) {
    alphaBoolBlend {blendOption=Linear}
    boneToModify {boneName=base_furCollarPiece_r_01}
    constraintSetup [{targetBone {boneName=spine_05}}, {targetBone {boneName=clavicle_r}}]
    constraintWeights [1,1]
   }
   n22 = "Constraint - base_furCollarPiece_r_02 (PPP)"!(ComponentPose=@n21, Alpha=1, ConstraintWeights_0=0.2, ConstraintWeights_1:float=0.5, ConstraintWeights_2:float=0.3) -> (Pose) {
    alphaBoolBlend {blendOption=Linear}
    boneToModify {boneName=base_furCollarPiece_r_02}
    constraintSetup [{targetBone {boneName=deform_upperarm_up_r}}, {targetBone {boneName=clavicle_r}}, {targetBone {boneName=spine_05}}]
    constraintWeights [1,1,1]
   }
   n23 = "Constraint - base_furCollarPiece_r_09 (PPPP)"!(
    ComponentPose=@n22
    Alpha=1
    ConstraintWeights_0=0.3
    ConstraintWeights_1:float=0.4
    ConstraintWeights_2:float=0.2
    ConstraintWeights_3:float=0.1
   ) -> (Pose) {
    alphaBoolBlend {blendOption=Linear}
    boneToModify {boneName=base_furCollarPiece_r_09}
    constraintSetup [
     {targetBone {boneName=deform_upperarm_up_r}}
     {targetBone {boneName=clavicle_r}}
     {targetBone {boneName=upperarm_r}}
     {targetBone {boneName=spine_05}}
    ]
    constraintWeights [1,1,1,1]
   }
   n24 = "Constraint - base_furCollarPiece_r_10 (PPP)"!(ComponentPose=@n23, Alpha=1, ConstraintWeights_0=0.3, ConstraintWeights_1:float=0.6, ConstraintWeights_2:float=0.1) -> (Pose) {
    alphaBoolBlend {blendOption=Linear}
    boneToModify {boneName=base_furCollarPiece_r_10}
    constraintSetup [{targetBone {boneName=deform_upperarm_bk_r}}, {targetBone {boneName=clavicle_r}}, {targetBone {boneName=upperarm_r}}]
    constraintWeights [1,1,1]
   }
   n25 = "Constraint - Corrective_Belt_Buckle (PPP)"!(ComponentPose=@n24, Alpha=1, ConstraintWeights_0=0.65, ConstraintWeights_1:float=0.175, ConstraintWeights_2:float=0.175) -> (Pose) {
    alphaBoolBlend {blendOption=Linear}
    boneToModify {boneName=Corrective_Belt_Buckle}
    constraintSetup [{targetBone {boneName=pelvis}}, {targetBone {boneName=thigh_l}}, {targetBone {boneName=thigh_r}}]
    constraintWeights [1,1,1]
   }
   Save cached pose 'Constraints2AnimDyn'(Pose=Component To Local(ComponentPose=@n25)) {cacheName=Constraints2AnimDyn}
   n26 = Anim Dynamics - Bone: dyn_furCollarPiece_l_01!(
    ComponentPose=Local To Component(LocalPose=Use cached pose 'Constraints2AnimDyn'())
    Alpha=1
    GravityOverride:Vector="0,0,-980"
    ExternalForce="0,0,0"
   ) -> (Pose) {
    angularDampingOverride=1
    angularSpringConstant=1000
    bAngularSpring=true
    bOverrideAngularDamping=true
    bUseGravityOverride=true
    gravityScale=0.35
    simulationSpace=RootRelative
    alphaBoolBlend {blendOption=Linear}
    alphaScaleBiasClamp {bInterpResult=true}
    boundBone {boneName=dyn_furCollarPiece_l_01}
    componentLinearAccScale=<1 1 1>
    componentLinearVelScale=<1 1 1>
    gravityOverride {z=-980}
    physicsBodyDefinitions [{
      boundBone {boneName=dyn_furCollarPiece_l_01}
      boxExtents=<20 30 30>
      constraintSetup {angularLimitsMax=<15 15 15>, angularLimitsMin=<-15 -15 -15>, angularTarget=<1 0 0>}
      localJointOffset=<4 0 0>
     }]
   }
   n27 = Anim Dynamics - Bone: dyn_furCollarPiece_l_02!(ComponentPose=@n26, Alpha=1, GravityOverride:Vector="0,0,-980", ExternalForce="0,0,0") -> (Pose) {
    angularDampingOverride=1
    angularSpringConstant=800
    bAngularSpring=true
    bOverrideAngularDamping=true
    bUseGravityOverride=true
    gravityScale=0.35
    simulationSpace=RootRelative
    alphaBoolBlend {blendOption=Linear}
    alphaScaleBiasClamp {bInterpResult=true}
    boundBone {boneName=dyn_furCollarPiece_l_02}
    componentLinearAccScale=<1 1 1>
    componentLinearVelScale=<1 1 1>
    gravityOverride {z=-980}
    physicsBodyDefinitions [{
      boundBone {boneName=dyn_furCollarPiece_l_02}
      boxExtents=<20 30 30>
      constraintSetup {angularLimitsMax=<15 15 15>, angularLimitsMin=<-15 -15 -15>, angularTarget=<1 0 0>}
      localJointOffset=<4 0 0>
     }]
   }
   n28 = Anim Dynamics - Bone: dyn_furCollarPiece_l_03!(ComponentPose=@n27, Alpha=1, GravityOverride:Vector="0,0,-980", ExternalForce="0,0,0") -> (Pose) {
    angularDampingOverride=1
    angularSpringConstant=450
    bAngularSpring=true
    bOverrideAngularDamping=true
    bUseGravityOverride=true
    gravityScale=0.35
    simulationSpace=RootRelative
    alphaBoolBlend {blendOption=Linear}
    alphaScaleBiasClamp {bInterpResult=true}
    boundBone {boneName=dyn_furCollarPiece_l_03}
    componentLinearAccScale=<1 1 1>
    componentLinearVelScale=<1 1 1>
    gravityOverride {z=-980}
    physicsBodyDefinitions [{
      boundBone {boneName=dyn_furCollarPiece_l_03}
      boxExtents=<20 30 30>
      constraintSetup {angularLimitsMax=<15 15 15>, angularLimitsMin=<-15 -15 0>, angularTarget=<1 0 0>}
      localJointOffset=<4 0 0>
     }]
   }
   n29 = Anim Dynamics - Bone: dyn_furCollarPiece_l_04!(ComponentPose=@n28, Alpha=1, GravityOverride:Vector="0,0,-980", ExternalForce="0,0,0") -> (Pose) {
    angularDampingOverride=1
    angularSpringConstant=1200
    bAngularSpring=true
    bOverrideAngularDamping=true
    bUseGravityOverride=true
    gravityScale=0.35
    simulationSpace=RootRelative
    alphaBoolBlend {blendOption=Linear}
    alphaScaleBiasClamp {bInterpResult=true}
    boundBone {boneName=dyn_furCollarPiece_l_04}
    componentLinearAccScale=<1 1 1>
    componentLinearVelScale=<1 1 1>
    gravityOverride {z=-980}
    physicsBodyDefinitions [{
      boundBone {boneName=dyn_furCollarPiece_l_04}
      boxExtents=<20 30 30>
      constraintSetup {angularLimitsMax=<15 0 15>, angularLimitsMin=<-15 -15 -15>, angularTarget=<1 0 0>}
      localJointOffset=<4 0 0>
     }]
   }
   n30 = Anim Dynamics - Bone: dyn_furCollarPiece_l_05!(ComponentPose=@n29, Alpha=1, GravityOverride:Vector="0,0,-980", ExternalForce="0,0,0") -> (Pose) {
    angularDampingOverride=1
    angularSpringConstant=1500
    bAngularSpring=true
    bOverrideAngularDamping=true
    bUseGravityOverride=true
    gravityScale=0.35
    simulationSpace=RootRelative
    alphaBoolBlend {blendOption=Linear}
    alphaScaleBiasClamp {bInterpResult=true}
    boundBone {boneName=dyn_furCollarPiece_l_05}
    componentLinearAccScale=<1 1 1>
    componentLinearVelScale=<1 1 1>
    gravityOverride {z=-980}
    physicsBodyDefinitions [{
      boundBone {boneName=dyn_furCollarPiece_l_05}
      boxExtents=<20 30 30>
      constraintSetup {angularLimitsMax=<15 15 15>, angularLimitsMin=<-15 -15 -15>, angularTarget=<1 0 0>}
      localJointOffset=<4 0 0>
     }]
   }
   n31 = Anim Dynamics - Bone: dyn_furCollarPiece_bl_01!(ComponentPose=@n30, Alpha=1, GravityOverride:Vector="0,0,-980", ExternalForce="0,0,0") -> (Pose) {
    angularDampingOverride=1
    angularSpringConstant=800
    bAngularSpring=true
    bOverrideAngularDamping=true
    bUseGravityOverride=true
    gravityScale=0.35
    simulationSpace=RootRelative
    alphaBoolBlend {blendOption=Linear}
    alphaScaleBiasClamp {bInterpResult=true}
    boundBone {boneName=dyn_furCollarPiece_bl_01}
    componentLinearAccScale=<1 1 1>
    componentLinearVelScale=<1 1 1>
    gravityOverride {z=-980}
    physicsBodyDefinitions [{
      boundBone {boneName=dyn_furCollarPiece_bl_01}
      boxExtents=<20 30 30>
      constraintSetup {angularLimitsMax=<15 5 15>, angularLimitsMin=<-15 -10 -15>, angularTarget=<1 0 0>}
      localJointOffset=<4 0 0>
     }]
   }
   n32 = Anim Dynamics - Bone: dyn_furCollarPiece_bl_02!(ComponentPose=@n31, Alpha=1, GravityOverride:Vector="0,0,-980", ExternalForce="0,0,0") -> (Pose) {
    angularDampingOverride=1
    angularSpringConstant=1000
    bAngularSpring=true
    bOverrideAngularDamping=true
    bUseGravityOverride=true
    gravityScale=0.35
    simulationSpace=RootRelative
    alphaBoolBlend {blendOption=Linear}
    alphaScaleBiasClamp {bInterpResult=true}
    boundBone {boneName=dyn_furCollarPiece_bl_02}
    componentLinearAccScale=<1 1 1>
    componentLinearVelScale=<1 1 1>
    gravityOverride {z=-980}
    physicsBodyDefinitions [{
      boundBone {boneName=dyn_furCollarPiece_bl_02}
      boxExtents=<20 30 30>
      constraintSetup {angularLimitsMax=<15 8 15>, angularLimitsMin=<-15 -15 -15>, angularTarget=<1 0 0>}
      localJointOffset=<4 0 0>
     }]
   }
   n33 = Anim Dynamics - Bone: dyn_furCollarPiece_bl_03!(ComponentPose=@n32, Alpha=1, GravityOverride:Vector="0,0,-980", ExternalForce="0,0,0") -> (Pose) {
    angularDampingOverride=1
    angularSpringConstant=1000
    bAngularSpring=true
    bOverrideAngularDamping=true
    bUseGravityOverride=true
    gravityScale=0.35
    simulationSpace=RootRelative
    alphaBoolBlend {blendOption=Linear}
    alphaScaleBiasClamp {bInterpResult=true}
    boundBone {boneName=dyn_furCollarPiece_bl_03}
    componentLinearAccScale=<1 1 1>
    componentLinearVelScale=<1 1 1>
    gravityOverride {z=-980}
    physicsBodyDefinitions [{
      boundBone {boneName=dyn_furCollarPiece_bl_03}
      boxExtents=<20 30 30>
      constraintSetup {angularLimitsMax=<15 8 15>, angularLimitsMin=<-15 -18 -15>, angularTarget=<1 0 0>}
      localJointOffset=<4 0 0>
     }]
   }
   n34 = Anim Dynamics - Bone: dyn_furCollarPiece_bl_04!(ComponentPose=@n33, Alpha=1, GravityOverride:Vector="0,0,-980", ExternalForce="0,0,0") -> (Pose) {
    angularDampingOverride=1
    angularSpringConstant=700
    bAngularSpring=true
    bOverrideAngularDamping=true
    bUseGravityOverride=true
    gravityScale=0.35
    simulationSpace=RootRelative
    alphaBoolBlend {blendOption=Linear}
    alphaScaleBiasClamp {bInterpResult=true}
    boundBone {boneName=dyn_furCollarPiece_bl_04}
    componentLinearAccScale=<1 1 1>
    componentLinearVelScale=<1 1 1>
    gravityOverride {z=-980}
    physicsBodyDefinitions [{
      boundBone {boneName=dyn_furCollarPiece_bl_04}
      boxExtents=<20 30 30>
      constraintSetup {angularLimitsMax=<15 8 15>, angularLimitsMin=<-15 -22 -15>, angularTarget=<1 0 0>}
      localJointOffset=<4 0 0>
     }]
   }
   n35 = Anim Dynamics - Bone: dyn_furCollarPiece_b_01!(ComponentPose=@n34, Alpha=1, GravityOverride:Vector="0,0,-980", ExternalForce="0,0,0") -> (Pose) {
    angularDampingOverride=1
    angularSpringConstant=750
    bAngularSpring=true
    bOverrideAngularDamping=true
    bUseGravityOverride=true
    gravityScale=0.35
    simulationSpace=RootRelative
    alphaBoolBlend {blendOption=Linear}
    alphaScaleBiasClamp {bInterpResult=true}
    boundBone {boneName=dyn_furCollarPiece_b_01}
    componentLinearAccScale=<1 1 1>
    componentLinearVelScale=<1 1 1>
    gravityOverride {z=-980}
    physicsBodyDefinitions [{
      boundBone {boneName=dyn_furCollarPiece_b_01}
      boxExtents=<20 30 30>
      constraintSetup {angularLimitsMax=<15 6 15>, angularLimitsMin=<-15 -25 -15>, angularTarget=<1 0 0>}
      localJointOffset=<4 0 0>
     }]
   }
   n36 = Anim Dynamics - Bone: dyn_furCollarPiece_b_02!(ComponentPose=@n35, Alpha=1, GravityOverride:Vector="0,0,-980", ExternalForce="0,0,0") -> (Pose) {
    angularDampingOverride=1
    angularSpringConstant=1100
    bAngularSpring=true
    bOverrideAngularDamping=true
    bUseGravityOverride=true
    gravityScale=0.35
    simulationSpace=RootRelative
    alphaBoolBlend {blendOption=Linear}
    alphaScaleBiasClamp {bInterpResult=true}
    boundBone {boneName=dyn_furCollarPiece_b_02}
    componentLinearAccScale=<1 1 1>
    componentLinearVelScale=<1 1 1>
    gravityOverride {z=-980}
    physicsBodyDefinitions [{
      boundBone {boneName=dyn_furCollarPiece_b_02}
      boxExtents=<20 30 30>
      constraintSetup {angularLimitsMax=<15 8 15>, angularLimitsMin=<-15 -15 -15>, angularTarget=<1 0 0>}
      localJointOffset=<4 0 0>
     }]
   }
   n37 = Anim Dynamics - Bone: dyn_furCollarPiece_b_03!(ComponentPose=@n36, Alpha=1, GravityOverride:Vector="0,0,-980", ExternalForce="0,0,0") -> (Pose) {
    angularDampingOverride=1
    angularSpringConstant=1300
    bAngularSpring=true
    bOverrideAngularDamping=true
    bUseGravityOverride=true
    gravityScale=0.35
    simulationSpace=RootRelative
    alphaBoolBlend {blendOption=Linear}
    alphaScaleBiasClamp {bInterpResult=true}
    boundBone {boneName=dyn_furCollarPiece_b_03}
    componentLinearAccScale=<1 1 1>
    componentLinearVelScale=<1 1 1>
    gravityOverride {z=-980}
    physicsBodyDefinitions [{
      boundBone {boneName=dyn_furCollarPiece_b_03}
      boxExtents=<20 30 30>
      constraintSetup {angularLimitsMax=<15 8 15>, angularLimitsMin=<-15 -20 -15>, angularTarget=<1 0 0>}
      localJointOffset=<4 0 0>
     }]
   }
   n38 = Anim Dynamics - Bone: dyn_furCollarPiece_b_04!(ComponentPose=@n37, Alpha=1, GravityOverride:Vector="0,0,-980", ExternalForce="0,0,0") -> (Pose) {
    angularDampingOverride=1
    angularSpringConstant=1400
    bAngularSpring=true
    bOverrideAngularDamping=true
    bUseGravityOverride=true
    gravityScale=0.35
    simulationSpace=RootRelative
    alphaBoolBlend {blendOption=Linear}
    alphaScaleBiasClamp {bInterpResult=true}
    boundBone {boneName=dyn_furCollarPiece_b_04}
    componentLinearAccScale=<1 1 1>
    componentLinearVelScale=<1 1 1>
    gravityOverride {z=-980}
    physicsBodyDefinitions [{
      boundBone {boneName=dyn_furCollarPiece_b_04}
      boxExtents=<20 30 30>
      constraintSetup {angularLimitsMax=<15 8 15>, angularLimitsMin=<-15 -18 -15>, angularTarget=<1 0 0>}
      localJointOffset=<4 0 0>
     }]
   }
   n39 = Anim Dynamics - Bone: dyn_furCollarPiece_br_01!(ComponentPose=@n38, Alpha=1, GravityOverride:Vector="0,0,-980", ExternalForce="0,0,0") -> (Pose) {
    angularDampingOverride=1
    angularSpringConstant=800
    bAngularSpring=true
    bOverrideAngularDamping=true
    bUseGravityOverride=true
    gravityScale=0.35
    simulationSpace=RootRelative
    alphaBoolBlend {blendOption=Linear}
    alphaScaleBiasClamp {bInterpResult=true}
    boundBone {boneName=dyn_furCollarPiece_br_01}
    componentLinearAccScale=<1 1 1>
    componentLinearVelScale=<1 1 1>
    gravityOverride {z=-980}
    physicsBodyDefinitions [{
      boundBone {boneName=dyn_furCollarPiece_br_01}
      boxExtents=<20 30 30>
      constraintSetup {angularLimitsMax=<15 8 15>, angularLimitsMin=<-15 -15 -15>, angularTarget=<1 0 0>}
      localJointOffset=<4 0 0>
     }]
   }
   n40 = Anim Dynamics - Bone: dyn_furCollarPiece_br_02!(ComponentPose=@n39, Alpha=1, GravityOverride:Vector="0,0,-980", ExternalForce="0,0,0") -> (Pose) {
    angularDampingOverride=1
    angularSpringConstant=1200
    bAngularSpring=true
    bOverrideAngularDamping=true
    bUseGravityOverride=true
    gravityScale=0.35
    simulationSpace=RootRelative
    alphaBoolBlend {blendOption=Linear}
    alphaScaleBiasClamp {bInterpResult=true}
    boundBone {boneName=dyn_furCollarPiece_br_02}
    componentLinearAccScale=<1 1 1>
    componentLinearVelScale=<1 1 1>
    gravityOverride {z=-980}
    physicsBodyDefinitions [{
      boundBone {boneName=dyn_furCollarPiece_br_02}
      boxExtents=<20 30 30>
      constraintSetup {angularLimitsMax=<15 8 15>, angularLimitsMin=<-15 -18 -15>, angularTarget=<1 0 0>}
      localJointOffset=<4 0 0>
     }]
   }
   n41 = Anim Dynamics - Bone: dyn_furCollarPiece_br_03!(ComponentPose=@n40, Alpha=1, GravityOverride:Vector="0,0,-980", ExternalForce="0,0,0") -> (Pose) {
    angularDampingOverride=1
    angularSpringConstant=1600
    bAngularSpring=true
    bOverrideAngularDamping=true
    bUseGravityOverride=true
    gravityScale=0.35
    simulationSpace=RootRelative
    alphaBoolBlend {blendOption=Linear}
    alphaScaleBiasClamp {bInterpResult=true}
    boundBone {boneName=dyn_furCollarPiece_br_03}
    componentLinearAccScale=<1 1 1>
    componentLinearVelScale=<1 1 1>
    gravityOverride {z=-980}
    physicsBodyDefinitions [{
      boundBone {boneName=dyn_furCollarPiece_br_03}
      boxExtents=<20 30 30>
      constraintSetup {angularLimitsMax=<15 8 15>, angularLimitsMin=<-15 -20 -15>, angularTarget=<1 0 0>}
      localJointOffset=<4 0 0>
     }]
   }
   n42 = Anim Dynamics - Bone: dyn_furCollarPiece_br_04!(ComponentPose=@n41, Alpha=1, GravityOverride:Vector="0,0,-980", ExternalForce="0,0,0") -> (Pose) {
    angularDampingOverride=1
    angularSpringConstant=900
    bAngularSpring=true
    bOverrideAngularDamping=true
    bUseGravityOverride=true
    gravityScale=0.35
    simulationSpace=RootRelative
    alphaBoolBlend {blendOption=Linear}
    alphaScaleBiasClamp {bInterpResult=true}
    boundBone {boneName=dyn_furCollarPiece_br_04}
    componentLinearAccScale=<1 1 1>
    componentLinearVelScale=<1 1 1>
    gravityOverride {z=-980}
    physicsBodyDefinitions [{
      boundBone {boneName=dyn_furCollarPiece_br_04}
      boxExtents=<20 30 30>
      constraintSetup {angularLimitsMax=<15 8 15>, angularLimitsMin=<-15 -15 -15>, angularTarget=<1 0 0>}
      localJointOffset=<4 0 0>
     }]
   }
   n43 = Anim Dynamics - Bone: dyn_furCollarPiece_br_05!(ComponentPose=@n42, Alpha=1, GravityOverride:Vector="0,0,-980", ExternalForce="0,0,0") -> (Pose) {
    angularDampingOverride=1
    angularSpringConstant=1500
    bAngularSpring=true
    bOverrideAngularDamping=true
    bUseGravityOverride=true
    gravityScale=0.35
    simulationSpace=RootRelative
    alphaBoolBlend {blendOption=Linear}
    alphaScaleBiasClamp {bInterpResult=true}
    boundBone {boneName=dyn_furCollarPiece_br_05}
    componentLinearAccScale=<1 1 1>
    componentLinearVelScale=<1 1 1>
    gravityOverride {z=-980}
    physicsBodyDefinitions [{
      boundBone {boneName=dyn_furCollarPiece_br_05}
      boxExtents=<20 30 30>
      constraintSetup {angularLimitsMax=<15 8 15>, angularLimitsMin=<-15 -15 -15>, angularTarget=<1 0 0>}
      localJointOffset=<4 0 0>
     }]
   }
   n44 = Anim Dynamics - Bone: dyn_furCollarPiece_br_06!(ComponentPose=@n43, Alpha=1, GravityOverride:Vector="0,0,-980", ExternalForce="0,0,0") -> (Pose) {
    angularDampingOverride=1
    angularSpringConstant=2000
    bAngularSpring=true
    bOverrideAngularDamping=true
    bUseGravityOverride=true
    gravityScale=0.35
    simulationSpace=RootRelative
    alphaBoolBlend {blendOption=Linear}
    alphaScaleBiasClamp {bInterpResult=true}
    boundBone {boneName=dyn_furCollarPiece_br_06}
    componentLinearAccScale=<1 1 1>
    componentLinearVelScale=<1 1 1>
    gravityOverride {z=-980}
    physicsBodyDefinitions [{
      boundBone {boneName=dyn_furCollarPiece_br_06}
      boxExtents=<20 30 30>
      constraintSetup {angularLimitsMax=<15 8 15>, angularLimitsMin=<-15 -25 -15>, angularTarget=<1 0 0>}
      localJointOffset=<4 0 0>
     }]
   }
   n45 = Anim Dynamics - Bone: dyn_furCollarPiece_r_01!(ComponentPose=@n44, Alpha=1, GravityOverride:Vector="0,0,-980", ExternalForce="0,0,0") -> (Pose) {
    angularDampingOverride=1
    angularSpringConstant=750
    bAngularSpring=true
    bOverrideAngularDamping=true
    bUseGravityOverride=true
    gravityScale=0.35
    simulationSpace=RootRelative
    alphaBoolBlend {blendOption=Linear}
    alphaScaleBiasClamp {bInterpResult=true}
    boundBone {boneName=dyn_furCollarPiece_r_01}
    componentLinearAccScale=<1 1 1>
    componentLinearVelScale=<1 1 1>
    gravityOverride {z=-980}
    physicsBodyDefinitions [{
      boundBone {boneName=dyn_furCollarPiece_r_01}
      boxExtents=<20 30 30>
      constraintSetup {angularLimitsMax=<15 15 15>, angularLimitsMin=<-15 -15 -15>, angularTarget=<1 0 0>}
      localJointOffset=<4 0 0>
     }]
   }
   n46 = Anim Dynamics - Bone: dyn_furCollarPiece_r_02!(ComponentPose=@n45, Alpha=1, GravityOverride:Vector="0,0,-980", ExternalForce="0,0,0") -> (Pose) {
    angularDampingOverride=1
    angularSpringConstant=800
    bAngularSpring=true
    bOverrideAngularDamping=true
    bUseGravityOverride=true
    gravityScale=0.35
    simulationSpace=RootRelative
    alphaBoolBlend {blendOption=Linear}
    alphaScaleBiasClamp {bInterpResult=true}
    boundBone {boneName=dyn_furCollarPiece_r_02}
    componentLinearAccScale=<1 1 1>
    componentLinearVelScale=<1 1 1>
    gravityOverride {z=-980}
    physicsBodyDefinitions [{
      boundBone {boneName=dyn_furCollarPiece_r_02}
      boxExtents=<20 30 30>
      constraintSetup {angularLimitsMax=<15 15 15>, angularLimitsMin=<-15 -15 -8>, angularTarget=<1 0 0>}
      localJointOffset=<4 0 0>
     }]
   }
   n47 = Anim Dynamics - Bone: dyn_furCollarPiece_r_03!(ComponentPose=@n46, Alpha=1, GravityOverride:Vector="0,0,-980", ExternalForce="0,0,0") -> (Pose) {
    angularDampingOverride=1
    angularSpringConstant=1000
    bAngularSpring=true
    bOverrideAngularDamping=true
    bUseGravityOverride=true
    gravityScale=0.35
    simulationSpace=RootRelative
    alphaBoolBlend {blendOption=Linear}
    alphaScaleBiasClamp {bInterpResult=true}
    boundBone {boneName=dyn_furCollarPiece_r_03}
    componentLinearAccScale=<1 1 1>
    componentLinearVelScale=<1 1 1>
    gravityOverride {z=-980}
    physicsBodyDefinitions [{
      boundBone {boneName=dyn_furCollarPiece_r_03}
      boxExtents=<20 30 30>
      constraintSetup {angularLimitsMax=<15 15 5>, angularLimitsMin=<-15 -15 -15>, angularTarget=<1 0 0>}
      localJointOffset=<4 0 0>
     }]
   }
   n48 = Anim Dynamics - Bone: dyn_furCollarPiece_r_04!(ComponentPose=@n47, Alpha=1, GravityOverride:Vector="0,0,-980", ExternalForce="0,0,0") -> (Pose) {
    angularDampingOverride=1
    angularSpringConstant=850
    bAngularSpring=true
    bOverrideAngularDamping=true
    bUseGravityOverride=true
    gravityScale=0.35
    simulationSpace=RootRelative
    alphaBoolBlend {blendOption=Linear}
    alphaScaleBiasClamp {bInterpResult=true}
    boundBone {boneName=dyn_furCollarPiece_r_04}
    componentLinearAccScale=<1 1 1>
    componentLinearVelScale=<1 1 1>
    gravityOverride {z=-980}
    physicsBodyDefinitions [{
      boundBone {boneName=dyn_furCollarPiece_r_04}
      boxExtents=<20 30 30>
      constraintSetup {angularLimitsMax=<15 8 15>, angularLimitsMin=<-15 -15 -15>, angularTarget=<1 0 0>}
      localJointOffset=<4 0 0>
     }]
   }
   n49 = Anim Dynamics - Bone: dyn_furCollarPiece_r_05!(ComponentPose=@n48, Alpha=1, GravityOverride:Vector="0,0,-980", ExternalForce="0,0,0") -> (Pose) {
    angularDampingOverride=1
    angularSpringConstant=1100
    bAngularSpring=true
    bOverrideAngularDamping=true
    bUseGravityOverride=true
    gravityScale=0.35
    simulationSpace=RootRelative
    alphaBoolBlend {blendOption=Linear}
    alphaScaleBiasClamp {bInterpResult=true}
    boundBone {boneName=dyn_furCollarPiece_r_05}
    componentLinearAccScale=<1 1 1>
    componentLinearVelScale=<1 1 1>
    gravityOverride {z=-980}
    physicsBodyDefinitions [{
      boundBone {boneName=dyn_furCollarPiece_r_05}
      boxExtents=<20 30 30>
      constraintSetup {angularLimitsMax=<15 8 15>, angularLimitsMin=<-15 -15 -15>, angularTarget=<1 0 0>}
      localJointOffset=<4 0 0>
     }]
   }
   n50 = Anim Dynamics - Bone: dyn_furCollarPiece_r_06!(ComponentPose=@n49, Alpha=1, GravityOverride:Vector="0,0,-980", ExternalForce="0,0,0") -> (Pose) {
    angularDampingOverride=1
    angularSpringConstant=2000
    bAngularSpring=true
    bOverrideAngularDamping=true
    bUseGravityOverride=true
    gravityScale=0.35
    simulationSpace=RootRelative
    alphaBoolBlend {blendOption=Linear}
    alphaScaleBiasClamp {bInterpResult=true}
    boundBone {boneName=dyn_furCollarPiece_r_06}
    componentLinearAccScale=<1 1 1>
    componentLinearVelScale=<1 1 1>
    gravityOverride {z=-980}
    physicsBodyDefinitions [{
      boundBone {boneName=dyn_furCollarPiece_r_06}
      boxExtents=<20 30 30>
      constraintSetup {angularLimitsMax=<15 15 15>, angularLimitsMin=<-15 -25 -8>, angularTarget=<1 0 0>}
      localJointOffset=<4 0 0>
     }]
   }
   n51 = Anim Dynamics - Bone: dyn_furCollarPiece_r_07!(ComponentPose=@n50, Alpha=1, GravityOverride:Vector="0,0,-980", ExternalForce="0,0,0") -> (Pose) {
    angularDampingOverride=1
    angularSpringConstant=1000
    bAngularSpring=true
    bOverrideAngularDamping=true
    bUseGravityOverride=true
    gravityScale=0.35
    simulationSpace=RootRelative
    alphaBoolBlend {blendOption=Linear}
    alphaScaleBiasClamp {bInterpResult=true}
    boundBone {boneName=dyn_furCollarPiece_r_07}
    componentLinearAccScale=<1 1 1>
    componentLinearVelScale=<1 1 1>
    gravityOverride {z=-980}
    physicsBodyDefinitions [{
      boundBone {boneName=dyn_furCollarPiece_r_07}
      boxExtents=<20 30 30>
      constraintSetup {angularLimitsMax=<15 8 15>, angularLimitsMin=<-15 -15 -15>, angularTarget=<1 0 0>}
      localJointOffset=<4 0 0>
     }]
   }
   n52 = Anim Dynamics - Bone: dyn_furCollarPiece_r_08!(ComponentPose=@n51, Alpha=1, GravityOverride:Vector="0,0,-980", ExternalForce="0,0,0") -> (Pose) {
    angularDampingOverride=1
    angularSpringConstant=700
    bAngularSpring=true
    bOverrideAngularDamping=true
    bUseGravityOverride=true
    gravityScale=0.35
    simulationSpace=RootRelative
    alphaBoolBlend {blendOption=Linear}
    alphaScaleBiasClamp {bInterpResult=true}
    boundBone {boneName=dyn_furCollarPiece_r_08}
    componentLinearAccScale=<1 1 1>
    componentLinearVelScale=<1 1 1>
    gravityOverride {z=-980}
    physicsBodyDefinitions [{
      boundBone {boneName=dyn_furCollarPiece_r_08}
      boxExtents=<20 30 30>
      constraintSetup {angularLimitsMax=<15 8 15>, angularLimitsMin=<-15 -20 -15>, angularTarget=<1 0 0>}
      localJointOffset=<4 0 0>
     }]
   }
   n53 = Anim Dynamics - Bone: dyn_furCollarPiece_r_09!(ComponentPose=@n52, Alpha=1, GravityOverride:Vector="0,0,-980", ExternalForce="0,0,0") -> (Pose) {
    angularDampingOverride=1
    angularSpringConstant=2200
    bAngularSpring=true
    bOverrideAngularDamping=true
    bUseGravityOverride=true
    gravityScale=0.35
    simulationSpace=RootRelative
    alphaBoolBlend {blendOption=Linear}
    alphaScaleBiasClamp {bInterpResult=true}
    boundBone {boneName=dyn_furCollarPiece_r_09}
    componentLinearAccScale=<1 1 1>
    componentLinearVelScale=<1 1 1>
    gravityOverride {z=-980}
    physicsBodyDefinitions [{
      boundBone {boneName=dyn_furCollarPiece_r_09}
      boxExtents=<20 30 30>
      constraintSetup {angularLimitsMax=<15 8 15>, angularLimitsMin=<-15 -25 -15>, angularTarget=<1 0 0>}
      localJointOffset=<4 0 0>
     }]
   }
   n54 = Anim Dynamics - Bone: dyn_furCollarPiece_r_10!(ComponentPose=@n53, Alpha=1, GravityOverride:Vector="0,0,-980", ExternalForce="0,0,0") -> (Pose) {
    angularDampingOverride=1
    angularSpringConstant=1000
    bAngularSpring=true
    bOverrideAngularDamping=true
    bUseGravityOverride=true
    gravityScale=0.35
    simulationSpace=RootRelative
    alphaBoolBlend {blendOption=Linear}
    alphaScaleBiasClamp {bInterpResult=true}
    boundBone {boneName=dyn_furCollarPiece_r_10}
    componentLinearAccScale=<1 1 1>
    componentLinearVelScale=<1 1 1>
    gravityOverride {z=-980}
    physicsBodyDefinitions [{
      boundBone {boneName=dyn_furCollarPiece_r_10}
      boxExtents=<20 30 30>
      constraintSetup {angularLimitsMax=<15 8 15>, angularLimitsMin=<-15 -15 -15>, angularTarget=<1 0 0>}
      localJointOffset=<4 0 0>
     }]
   }
   Save cached pose 'AnimDyn2Logic'(Pose=Component To Local(ComponentPose=@n54)) {cacheName=AnimDyn2Logic}
  ]}
 EventGraph {nodes [
   Event BlueprintUpdateAnimation()
   > n1 = Sequence!(execute) -> (then_0, then_1, then_2:Exec, then_3:Exec)
   n2 = Gravity Override State Function DEPRECATED(
    execute=@n1.then_1
    CharacterPartAnimInstance=Self-Reference()
    GameTimeInSeconds=GetGameTimeInSeconds()
    GravityOverrideParams=Get Gravity Override Params()
   )
   > Set Gravity Override(Gravity Override=@n2)
   Set SkelMesh(execute=@n1.then_0, SkelMesh=GetPartSkeletalMeshComponent())
   n3 = NOT Boolean(A=Get bIsPlayingEmote())
   n4 = GetAlphaFromReferenceTransformDelta(SkeletalMeshComponent=Get SkelMesh(), BoneName=thigh_l, BaseName=calf_l, InputRange=(X=-45,Y=-120), TwistAxis="0,0,1")
   Set Calf L Rot(
    execute=@n1.then_2
    Calf L Rot:float=float * float!(A=@n4, B=@n4, C:double="To Float (Boolean)"(InBool=@n3)) -> (ReturnValue) {is=title K2Node_PromotableOperator}
   ) -> (Output_Get:float)
   n5 = GetAlphaFromReferenceTransformDelta(SkeletalMeshComponent=Get SkelMesh(), BoneName=thigh_r, BaseName=calf_r, InputRange=(X=-45,Y=-120), TwistAxis="0,0,1")
   n6 = NOT Boolean(A=Get bIsPlayingEmote())
   > Set Calf R Rot(Calf R Rot:float=float * float(A=@n5, B=@n5, C="To Float (Boolean)"(InBool=@n6))) -> (Output_Get:float)
   n7 = Get bIsDBNO()
   > "Set Is DBNO "("Is DBNO ":double=float * float!(A=Get Calf L Rot(), B="To Float (Boolean)"(InBool=@n7)) -> (ReturnValue)) -> (Output_Get:double)
   Set Head Up(
    execute=@n1.then_3
    Head Up:float=GetAlphaFromReferenceTransformDelta(SkeletalMeshComponent=Get SkelMesh(), BoneName=spine_05, BaseName=head, InputRange=(X=0,Y=110), TwistAxis="0,0,1")
   ) -> (Output_Get:float)
   > Set Head Right(Head Right:float=GetAlphaFromReferenceTransformDelta(SkeletalMeshComponent=Get SkelMesh(), BoneName=spine_05, BaseName=head, InputRange=(X=0,Y=110), TwistAxis="1,0,0")) -> (Output_Get:float)
   > Set Head Left(Head Left:float=GetAlphaFromReferenceTransformDelta(SkeletalMeshComponent=Get SkelMesh(), BoneName=spine_05, BaseName=head, InputRange=(X=0,Y=-110), TwistAxis="1,0,0")) -> (Output_Get:float)
  ]}
]
variables [
 Calf L Rot {type=number}
 Calf R Rot {type=number}
 Gravity Override {type=Vector}
 Gravity Override Params {type=GravityOverrideParamsStruct}
 Head Left {type=number}
 Head Right {type=number}
 Head Up {type=number}
 "Is DBNO " {type=number}
 SkelMesh {type=SkeletalMeshComponent}
]
